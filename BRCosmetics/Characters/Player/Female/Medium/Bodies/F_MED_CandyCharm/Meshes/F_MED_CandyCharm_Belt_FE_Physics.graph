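# Physics asset for the belt of the CandyCharm female medium cosmetic body, built entirely from capsule bodies.
# A five-capsule dyn_Belt_MSTR_01 body is the hub, constrained to the master, two mid and left belt bodies with very tight swing limits.
# The back belt piece hangs separately off its base bone, while pelvis and right thigh capsules form their own looser constrained pair.
# All constraints have zero twist limits.
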PhysicsAsset /BRCosmetics/Characters/Player/Female/Medium/Bodies/F_MED_CandyCharm/Meshes/F_MED_CandyCharm_Belt_FE_Physics
bodies {
 dyn_Belt_Back_01 [{
   shapeType=Capsule
   length=3.713
   radius=2.2969
   shapeName=dyn_Belt_Back_01_capsule
   center=<-0.0845 1.458 -0.0137>
   rotation=<rot 3.9833 -64.4716 81.7245>
  }]
 dyn_Belt_MSTR [{shapeType=Capsule, length=0.505, radius=0.505, shapeName=dyn_Belt_MSTR_capsule, rotation=<rot 0 -90 0>}]
 dyn_Belt_MSTR_01 [
  {
   shapeType=Capsule
   length=7.5525
   radius=3.0013
   shapeName=dyn_Belt_MSTR_01_capsule
   center=<36.6159 1.6804 1.1857>
   rotation=<rot -62.8994 -0.0001 -90.2867>
  }
  {
   shapeType=Capsule
   length=16.982
   radius=3.2734
   shapeName=dyn_Belt_MSTR_01_capsule
   center=<27.7994 14.9687 0.0026>
   rotation=<rot -80.1168 5.9595 -44.6727>
  }
  {
   shapeType=Capsule
   length=10.3567
   radius=3.4395
   shapeName=dyn_Belt_MSTR_01_capsule
   center=<36.101 -5.6268 2.1884>
   rotation=<rot 3.8381 -36.4615 -93.1411>
  }
  {
   shapeType=Capsule
   length=13.8491
   radius=2.6862
   shapeName=dyn_Belt_MSTR_01_capsule
   center=<14.3211 16.9613 -0.4119>
   rotation=<rot -77.8876 -137.2781 156.3894>
  }
  {
   shapeType=Capsule
   length=12.625
   radius=4.125
   shapeName=dyn_Belt_MSTR_01_capsule
   center=<20.117 14.3913 -0.1497>
   rotation=<rot -85 -0.0002 0.0004>
  }
 ]
 dyn_Belt_Mid_01 [{
   shapeType=Capsule
   length=6
   radius=3
   shapeName=dyn_Belt_Mid_01_capsule
   center=<0.1196 -0.6785 0.0001>
   rotation=<rot 0.0084 -80.0024 -84.4281>
  }]
 dyn_Belt_Mid_02 [{
   shapeType=Capsule
   length=7.125
   radius=4.875
   shapeName=dyn_Belt_Mid_02_capsule
   center=<-0.5609 -3.0826 -0.1532>
   rotation=<rot -3.6744 -100.1628 80.3269>
  }]
 dyn_Belt_l_01 [{
   shapeType=Capsule
   length=10
   radius=3
   shapeName=dyn_Belt_l_01_capsule
   center=<2.7729 5.3706 0.4518>
   rotation=<rot -21.1421 -159.2452 -87.5751>
  }]
 dyn_base_Belt_Back [{shapeType=Capsule, length=0.505, radius=0.505, shapeName=dyn_base_Belt_Back_capsule, rotation=<rot 0 -90 0>}]
 pelvis [{shapeType=Capsule, length=14.8113, radius=11.2417, shapeName=pelvis_capsule, center=<-5.6005 -2.1275 0>, rotation=<rot 6.1186 -0.635 0.1501>}]
 thigh_r [{
   shapeType=Capsule
   length=28.4611
   radius=10.9374
   shapeName=thigh_r_capsule
   center=<16.6531 0.5669 -0.7141>
   rotation=<rot -88.7831 -0.0022 -6.1513>
  }]
}
constraints [
 {bone1Name=thigh_r, bone2Name=pelvis, swing1LimitDegrees=20, swing2LimitDegrees=20, twistLimitDegrees=0}
 {bone1Name=dyn_Belt_MSTR_01, bone2Name=dyn_Belt_MSTR, swing1LimitDegrees=2, swing2LimitDegrees=2, twistLimitDegrees=0}
 {bone1Name=dyn_Belt_Back_01, bone2Name=dyn_base_Belt_Back, swing1LimitDegrees=5, swing2LimitDegrees=5, twistLimitDegrees=0}
 {bone1Name=dyn_Belt_MSTR_01, bone2Name=dyn_Belt_Mid_01, swing1LimitDegrees=2, swing2LimitDegrees=2, twistLimitDegrees=0}
 {bone1Name=dyn_Belt_MSTR_01, bone2Name=dyn_Belt_Mid_02, swing1LimitDegrees=2, swing2LimitDegrees=2, twistLimitDegrees=0}
 {bone1Name=dyn_Belt_MSTR_01, bone2Name=dyn_Belt_l_01, swing1LimitDegrees=2, swing2LimitDegrees=2, twistLimitDegrees=0}
]
