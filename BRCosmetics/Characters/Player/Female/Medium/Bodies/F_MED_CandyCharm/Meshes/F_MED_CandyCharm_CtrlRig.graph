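# Control Rig for the F_MED_CandyCharm female body that applies corrective adjustments to deform bones and outfit dyn bones.
# The shared DeformRig function runs first, then a Sequence splits the collar, jacket and belt passes.
# Spherical pose readers on upper arms (against spine_05) and pelvis (against each thigh) weight additive-style ModifyTransforms on collar and belt bones.
# The SkyDive_Back_ADJ variable weights the jacket back offset.
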ControlRigBlueprint /BRCosmetics/Characters/Player/Female/Medium/Bodies/F_MED_CandyCharm/Meshes/F_MED_CandyCharm_CtrlRig
graphs [
 RigVMFunctionLibrary {links [], nodes []}
 RigVMFunction_Sequence_1_ContainedGraph {nodes [
   n1 = RigVMFunctionEntryNode()
   RigVMFunction_Sequence = &RigVMFunction_Sequence(ExecuteContext=@n1.ExecuteContext)
   n2 = &RigVMFunction_Sequence(ExecuteContext=@RigVMFunction_Sequence.B)
   RigVMFunctionReturnNode(A=@RigVMFunction_Sequence.A, B=@n2.A, C=@n2.B)
  ]}
 RigVMModel {nodes [
   RigUnit_BeginExecution = &RigUnit_BeginExecution()
   FN_DeformRig_FNC = RigVMFunctionReferenceNode(
    DeformRig_ Curve="(deformRig_l_upperarm_up_alpha_72_#1=False,deformRig_r_upperarm_up_alpha_73_#2=False,deformRig_l_upperarm_fr_alpha_74_#3=False,deformRig_r_upperarm_fr_alpha_75_CB51B8 ...(truncated)"
    DeformRig_Bool="(upperarm_1_#4=True,elbow_3_#5=True,wrist_7_#6=True,groin_8_#7=True,glute_17_#8=True,knee_9_40C408324CA261 ...(truncated)"
    DeformRig_Debug="(upperarm_debug=False,wrist_debug=False,lat_debug=False,groinGlute_debug=False,upperarm_debug_18_#4=False,wrist_debug_20_#6=False,groinGlute_debug_27_#7=False,lat_deb ...(truncated)"
    DeformRig_Value="(upperarm_up_18_#9=(Rotation=(X=0.0,Y=-0.0,Z=0.0,W=1.0),Translation=(X=0.0,Y=0.0,Z=8.0),Scale3D=(X=0.70,Y=1.0,Z=1.0)),upperarm_fwd_fr_19_#10=(Rotation=(X=0.0,Y=-0.0,Z=0.0,W=1.0),Translation=(X=0.0, ...(truncated)"
    ExecuteContext=@RigUnit_BeginExecution.ExecutePin
   )
   > RigVMFunction_Sequence = RigVMAggregateNode()
   &RigUnit_ModifyTransforms(
    ItemToModify=((Item=(Type=Bone,Name="upperarm_r"),Transform=(Rotation=(X=0,Y=0,Z=0,W=1),Translation=(X=0,Y=0,Z=0),Scale3D=(X=1,Y=1,Z=1))))
    Weight=1
    WeightMinimum=0
    ExecutePin=@RigVMFunction_Sequence.A
   )
   > n1 = &RigUnit_SphericalPoseReader(
    ActiveRegionScaleFactors=(PositiveWidth=1,NegativeWidth=1,PositiveHeight=2,NegativeHeight=2)
    ActiveRegionSize=0.1
    Debug=(bDrawDebug=false,bDraw2D=False,bDrawLocalAxes=False,DebugScale=25,DebugSegments=20,DebugThickness=0.25)
    DriverAxis=(X=-1,Y=0,Z=0)
    DriverItem=(Type=Bone,Name="upperarm_r")
    FalloffRegionScaleFactors=(PositiveWidth=1,NegativeWidth=1,PositiveHeight=3,NegativeHeight=2)
    FalloffSize=0.3
    FlipHeightScaling=False
    FlipWidthScaling=False
    OptionalParentItem=(Type=Bone,Name="spine_05")
    RotationOffset=(X=-120,Y=30,Z=20)
   )
   > &RigUnit_ModifyTransforms(
    ItemToModify=((Item=(Type=Bone,Name="dyn_collar_front_01_R"),Transform=(Rotation=(X=0,Y=0.0872,Z=0,W=0.9962),Translation=(X=-0.5903,Y=0.2036,Z=-0.7134),Scale3D=(X=1,Y=1,Z=1))))
    WeightMinimum=0
    Weight=@n1.OutputParam
   )
   > &RigUnit_ModifyTransforms(
    ItemToModify=((Item=(Type=Bone,Name="dyn_collar_front_02_R"),Transform=(Rotation=(X=0,Y=-0.1045,Z=0,W=0.9945),Translation=(X=0.0596,Y=0.2036,Z=0.075),Scale3D=(X=1,Y=1,Z=1))))
    WeightMinimum=0
    Weight=@n1.OutputParam
   )
   > &RigUnit_ModifyTransforms(
    ItemToModify=((Item=(Type=Bone,Name="upperarm_l"),Transform=(Rotation=(X=0,Y=0,Z=0,W=1),Translation=(X=0,Y=0,Z=0),Scale3D=(X=1,Y=1,Z=1))))
    Weight=1
    WeightMinimum=0
   )
   > n2 = &RigUnit_SphericalPoseReader(
    ActiveRegionScaleFactors=(PositiveWidth=0.5,NegativeWidth=1.2,PositiveHeight=1,NegativeHeight=1)
    ActiveRegionSize=0.1
    Debug=(bDrawDebug=false,bDraw2D=False,bDrawLocalAxes=False,DebugScale=25,DebugSegments=20,DebugThickness=0.25)
    DriverAxis=(X=1,Y=0,Z=0)
    DriverItem=(Type=Bone,Name="upperarm_l")
    FalloffRegionScaleFactors=(PositiveWidth=0.3,NegativeWidth=0.2,PositiveHeight=2,NegativeHeight=1)
    FalloffSize=0.3
    FlipHeightScaling=False
    FlipWidthScaling=false
    OptionalParentItem=(Type=Bone,Name="spine_05")
    RotationOffset=(X=100,Y=0,Z=0)
   )
   > &RigUnit_ModifyTransforms(
    ItemToModify=((Item=(Type=Bone,Name="dyn_collar_front_01_L"),Transform=(Rotation=(X=0.0847,Y=0.2326,Z=-0.0203,W=0.9687),Translation=(X=0,Y=0,Z=0),Scale3D=(X=1,Y=1,Z=1))))
    WeightMinimum=0
    Weight=@n2.OutputParam
   )
   > &RigUnit_ModifyTransforms(
    ItemToModify=((Item=(Type=Bone,Name="dyn_collar_front_02_L"),Transform=(Rotation=(X=0.0847,Y=0.2326,Z=-0.0203,W=0.9687),Translation=(X=0,Y=0,Z=0),Scale3D=(X=1,Y=1,Z=1))))
    WeightMinimum=0
    Weight=@n2.OutputParam
   )
   > &RigUnit_ModifyTransforms(
    ItemToModify=((Item=(Type=Bone,Name="dyn_collar_side_01_L"),Transform=(Rotation=(X=0.0847,Y=0.2326,Z=-0.0203,W=0.9687),Translation=(X=0,Y=0,Z=0),Scale3D=(X=1,Y=1,Z=1))))
    WeightMinimum=0
    Weight=@n2.OutputParam
   )
   > n3 = &RigUnit_SphericalPoseReader(
    ActiveRegionScaleFactors=(PositiveWidth=0.8,NegativeWidth=1.2,PositiveHeight=1,NegativeHeight=1)
    ActiveRegionSize=0.1
    Debug=(bDrawDebug=false,bDraw2D=False,bDrawLocalAxes=False,DebugScale=25,DebugSegments=20,DebugThickness=0.25)
    DriverAxis=(X=1,Y=0,Z=0)
    DriverItem=(Type=Bone,Name="upperarm_l")
    FalloffRegionScaleFactors=(PositiveWidth=0.5,NegativeWidth=0.2,PositiveHeight=2,NegativeHeight=1)
    FalloffSize=0.3
    FlipHeightScaling=False
    FlipWidthScaling=False
    OptionalParentItem=(Type=Bone,Name="spine_05")
    RotationOffset=(X=30,Y=40,Z=0)
   )
   > &RigUnit_ModifyTransforms(
    ItemToModify=((Item=(Type=Bone,Name="dyn_collar_front_01_L"),Transform=(Rotation=(X=-0.0819,Y=0.3407,Z=0.0298,W=0.9361),Translation=(X=0,Y=0,Z=0),Scale3D=(X=1,Y=1,Z=1))))
    WeightMinimum=0
    Weight=@n3.OutputParam
   )
   > &RigUnit_ModifyTransforms(
    ItemToModify=((Item=(Type=Bone,Name="dyn_collar_side_01_L"),Transform=(Rotation=(X=0.0842,Y=0.2578,Z=-0.0226,W=0.9623),Translation=(X=0,Y=0,Z=0),Scale3D=(X=1,Y=1,Z=1))))
    WeightMinimum=0
    Weight=@n3.OutputParam
   )
   n4 = &RigUnit_SphericalPoseReader(
    ActiveRegionScaleFactors=(PositiveWidth=1,NegativeWidth=0.5,PositiveHeight=1.5,NegativeHeight=1.5)
    ActiveRegionSize=0.1
    Debug=(bDrawDebug=false,bDraw2D=False,bDrawLocalAxes=False,DebugScale=25,DebugSegments=20,DebugThickness=0.25)
    DriverAxis=(X=-1,Y=0,Z=0)
    DriverItem=(Type=Bone,Name="upperarm_r")
    FalloffRegionScaleFactors=(PositiveWidth=0.5,NegativeWidth=0.5,PositiveHeight=1.2,NegativeHeight=1)
    FalloffSize=0.3
    FlipHeightScaling=False
    FlipWidthScaling=False
    OptionalParentItem=(Type=Bone,Name="spine_05")
    RotationOffset=(X=-130,Y=70,Z=10)
    ExecutePin=@RigVMFunction_Sequence.B
   )
   > &RigUnit_ModifyTransforms(
    ItemToModify=((Item=(Type=Bone,Name="dyn_collar_front_01_R"),Transform=(Rotation=(X=-0.3463,Y=0.3421,Z=0.1373,W=0.8626),Translation=(X=0,Y=0,Z=0),Scale3D=(X=1,Y=1,Z=1))))
    WeightMinimum=0
    Weight=@n4.OutputParam
   )
   > &RigUnit_ModifyTransforms(
    ItemToModify=((Item=(Type=Bone,Name="dyn_collar_front_02_R"),Transform=(Rotation=(X=0,Y=0.1058,Z=0,W=-0.9944),Translation=(X=0,Y=0,Z=0),Scale3D=(X=1,Y=1,Z=1))))
    WeightMinimum=0
    Weight=@n4.OutputParam
   )
   > &RigUnit_ModifyTransforms(
    ItemToModify="((Item=(Type=Bone,Name=\"dyn_jacket_back_01_L\"),Transform=(Rotation=(X=0,Y=0.0131,Z=0,W=0.9999),Translation=(X=0,Y=0,Z=1.5),Scale3D=(X=1,Y=1,Z=1))),(Item=(Type=Bone,Name=\"dyn_base_jacket_back_R\"),Transform=(Rotation=(X=0,Y=-0.0131,Z=0,W=0.9999),Translation= ...(truncated)"
    WeightMinimum=0
    Weight=RigVMVariableNode() {variable=SkyDive_Back_ADJ}.Value
   )
   n5 = &RigUnit_SphericalPoseReader(
    ActiveRegionScaleFactors=(PositiveWidth=1,NegativeWidth=1,PositiveHeight=1,NegativeHeight=1)
    ActiveRegionSize=0.2
    Debug=(bDrawDebug=false,bDraw2D=False,bDrawLocalAxes=False,DebugScale=25,DebugSegments=20,DebugThickness=0.25)
    DriverAxis=(X=0,Y=1,Z=1)
    DriverItem=(Type=Bone,Name="pelvis")
    FalloffRegionScaleFactors=(PositiveWidth=1.2,NegativeWidth=1,PositiveHeight=1.2,NegativeHeight=1)
    FalloffSize=0.3
    FlipHeightScaling=False
    FlipWidthScaling=False
    OptionalParentItem=(Type=Bone,Name="thigh_r")
    RotationOffset=(X=-14.9393,Y=77.1255,Z=-11.7736)
    ExecutePin=@RigVMFunction_Sequence.C
   )
   > &RigUnit_ModifyTransforms(
    ItemToModify=((Item=(Type=Bone,Name="dyn_Belt_MSTR"),Transform=(Rotation=(X=0,Y=-0.0872,Z=0,W=0.9962),Translation=(X=0,Y=0,Z=0),Scale3D=(X=1,Y=1,Z=1))))
    WeightMinimum=0
    Weight=@n5.OutputParam
   )
   > &RigUnit_ModifyTransforms(
    ItemToModify=((Item=(Type=Bone,Name="dyn_Belt_MSTR_01"),Transform=(Rotation=(X=0,Y=0,Z=0,W=1),Translation=(X=0,Y=0,Z=0),Scale3D=(X=1,Y=1,Z=1))))
    WeightMinimum=0
    Weight=@n5.OutputParam
   )
   > &RigUnit_ModifyTransforms(
    ItemToModify="((Item=(Type=Bone,Name=\"dyn_base_Belt_Mid\"),Transform=(Rotation=(X=0,Y=0,Z=0,W=1),Translation=(X=0,Y=0,Z=5),Scale3D=(X=1,Y=1,Z=1))),(Item=(Type=Bone,Name=\"dyn_base_Belt_Mid_02\"),Transform=(Rotation=(X=0,Y=0,Z=0,W=1),Translation=(X=0,Y=0,Z=5),Scale3D=(X=1,Y ...(truncated)"
    WeightMinimum=0
    Weight=@n5.OutputParam
   )
   > &RigUnit_ModifyTransforms(
    ItemToModify=((Item=(Type=Bone,Name="dyn_Belt_Mid_02"),Transform=(Rotation=(X=0,Y=0,Z=0,W=1),Translation=(X=0,Y=0,Z=0),Scale3D=(X=1,Y=1,Z=1))))
    WeightMinimum=0
    Weight=@n5.OutputParam
   )
   > n6 = &RigUnit_SphericalPoseReader(
    ActiveRegionScaleFactors=(PositiveWidth=1,NegativeWidth=1,PositiveHeight=1,NegativeHeight=1)
    ActiveRegionSize=0.1
    Debug=(bDrawDebug=false,bDraw2D=False,bDrawLocalAxes=False,DebugScale=25,DebugSegments=20,DebugThickness=0.25)
    DriverAxis=(X=0,Y=1,Z=-1)
    DriverItem=(Type=Bone,Name="pelvis")
    FalloffRegionScaleFactors=(PositiveWidth=1,NegativeWidth=1.8,PositiveHeight=1,NegativeHeight=1)
    FalloffSize=0.3
    FlipHeightScaling=False
    FlipWidthScaling=False
    OptionalParentItem=(Type=Bone,Name="thigh_l")
    RotationOffset=(X=10,Y=105,Z=-10)
   )
   > &RigUnit_ModifyTransforms(
    ItemToModify=((Item=(Type=Bone,Name="dyn_Belt_MSTR"),Transform=(Rotation=(X=0,Y=0.1305,Z=0,W=0.9914),Translation=(X=0,Y=0,Z=6),Scale3D=(X=1,Y=1,Z=1))))
    WeightMinimum=0
    Weight=@n6.OutputParam
   )
   > &RigUnit_ModifyTransforms(
    ItemToModify="((Item=(Type=Bone,Name=\"dyn_base_Belt_Front\"),Transform=(Rotation=(X=0,Y=0,Z=0,W=1),Translation=(X=0,Y=0,Z=2),Scale3D=(X=1,Y=1,Z=1))),(Item=(Type=Bone,Name=\"dyn_base_Belt_Front_L\"),Transform=(Rotation=(X=0,Y=0,Z=0,W=1),Translation=(X=0,Y=0,Z=2),Scale3D=(X= ...(truncated)"
    WeightMinimum=0
    Weight=@n6.OutputParam
   )
  ]}
]
hierarchy [
 root {loc=<0 0 0>, rot=<0 0 0 1>}
 pelvis {parent=root, loc=<0 -0.5301 98.0571>, rot=<0.0127 0.707 -0.0127 -0.707>}
 spine_01 {parent=pelvis, loc=<6.1208 -0.9328 0>, rot=<0 0 0.034 -0.9994>}
 spine_02 {parent=spine_01, loc=<6.5997 -0.2095 0>, rot=<0 0 0.0238 -0.9997>}
 spine_03 {parent=spine_02, loc=<6.4914 -0.1237 0>, rot=<0 0 0.0149 -0.9999>}
 spine_04 {parent=spine_03, loc=<6.9598 0.0794 0>, rot=<0 0 -0.0132 -0.9999>}
 spine_05 {parent=spine_04, loc=<6.9987 0 0>, rot=<0 0 -0.0015 -1>}
 neck_01 {parent=spine_05, loc=<11.3138 0.1529 0>, rot=<0 0 -0.2003 -0.9797>}
 neck_02 {parent=neck_01, loc=<4.5627 0 0>, rot=<0 0 0.0504 -0.9987>}
 head {parent=neck_02, loc=<5.3914 0 0>, rot=<0 0 0.115 -0.9934>}
 FACIAL_C_Jawline {parent=head, loc=<-6.5438 5.904 0.0002>, rot=<0.6934 -0.1389 -0.6933 -0.1389>}
 FACIAL_C_UnderChin {parent=head, loc=<-6.5027 3.7584 0.0012>, rot=<0.6786 -0.1988 -0.6785 -0.1988>}
 base_hair {parent=head, loc=<12.6041 -4.6143 -0.0849>, rot=<-0.5978 0.3121 0.2943 -0.6772>}
 dyn_hair_01 {parent=base_hair, loc=<0 0 0>, rot=<0 0 0 1>}
 dyn_hair_02 {parent=dyn_hair_01, loc=<7.9625 0 0>, rot=<0 0.5453 0 -0.8383>}
 dyn_hair_a_01_l {parent=dyn_hair_02, loc=<10.1023 -2.8389 6.391>, rot=<-0.4983 0.3818 -0.4648 -0.6244>}
 dyn_hair_a_02_l {parent=dyn_hair_a_01_l, loc=<7.0464 0 0>, rot=<0 0 0 1>}
 dyn_hair_b_01_l {parent=dyn_hair_02, loc=<6.2133 -3.5691 5.8719>, rot=<-0.568 0.4041 -0.348 -0.6268>}
 dyn_hair_b_02_l {parent=dyn_hair_b_01_l, loc=<6.5844 0 0>, rot=<0 0 0 1>}
 dyn_hair_c_01_l {parent=dyn_hair_02, loc=<10.9333 -4.0117 -0.9904>, rot=<-0.5636 0.4103 -0.3548 -0.623>}
 dyn_hair_c_02_l {parent=dyn_hair_c_01_l, loc=<6.9732 0 0>, rot=<0 0 -0.0185 -0.9998>}
 dyn_hair_c_03_l {parent=dyn_hair_c_02_l, loc=<5.3769 0 0>, rot=<0 0 -0.0185 -0.9998>}
 dyn_hair_d_01_l {parent=dyn_hair_02, loc=<9.8141 -5.9468 1.6485>, rot=<-0.5857 0.4391 -0.3146 -0.6043>}
 dyn_hair_d_02_l {parent=dyn_hair_d_01_l, loc=<4.6584 0 0>, rot=<0 0 -0.0154 -0.9999>}
 dyn_hair_d_03_l {parent=dyn_hair_d_02_l, loc=<4.815 0 0>, rot=<0 0 -0.0154 -0.9999>}
 dyn_hair_a_01_m {parent=dyn_hair_02, loc=<12.0426 -1.0479 4.777>, rot=<-0.5564 0.42 -0.3655 -0.6168>}
 dyn_hair_a_02_m {parent=dyn_hair_a_01_m, loc=<5.7516 0.0319 0>, rot=<0 0 -0.0439 -0.999>}
 dyn_hair_a_03_m {parent=dyn_hair_a_02_m, loc=<5.5435 0.0426 0>, rot=<0 0 -0.0254 -0.9997>}
 dyn_hair_b_01_m {parent=dyn_hair_02, loc=<11.5624 0.1081 -0.7602>, rot=<-0.5546 0.3724 -0.3692 -0.6461>}
 dyn_hair_b_02_m {parent=dyn_hair_b_01_m, loc=<9.4903 0 0>, rot=<0.0009 0 -0.0097 -1>}
 dyn_hair_b_01a_r {parent=dyn_hair_02, loc=<-0.9478 2.419 4.1781>, rot=<-0.4505 -0.2427 -0.12 -0.8508>}
 dyn_hair_b_02a_r {parent=dyn_hair_b_01a_r, loc=<6.5279 0 0>, rot=<0.1059 0.217 -0.4256 -0.8721>}
 dyn_hair_b_01_r {parent=dyn_hair_b_02a_r, loc=<6.2775 0 0>, rot=<-0.0078 0.1529 -0.0502 -0.9869>}
 dyn_hair_b_02_r {parent=dyn_hair_b_01_r, loc=<5.6701 0 0>, rot=<0 0 0 1>}
 dyn_hair_a_01a_r {parent=dyn_hair_02, loc=<4.1237 6.8222 -2.3924>, rot=<-0.7019 0.2412 -0.2585 -0.6184>}
 dyn_hair_a_01_r {parent=dyn_hair_a_01a_r, loc=<8.395 0 0>, rot=<-0.0034 -0.0181 -0.1835 -0.9828>}
 dyn_hair_a_02_r {parent=dyn_hair_a_01_r, loc=<5.2287 0 0>, rot=<0 0 0 1>}
 dyn_hair_a_03_r {parent=dyn_hair_a_02_r, loc=<5.6022 0 0>, rot=<0 0 0 1>}
 dyn_hair_d_01_r {parent=dyn_hair_02, loc=<1.8106 6.1179 -0.4868>, rot=<-0.8365 -0.0345 -0.2197 -0.5008>}
 dyn_hair_d_02_r {parent=dyn_hair_d_01_r, loc=<6.9102 0 0>, rot=<0 0 -0.3246 -0.9459>}
 dyn_hair_d_03_r {parent=dyn_hair_d_02_r, loc=<6.4543 0 0>, rot=<-0.0442 0.0018 -0.0411 -0.9982>}
 dyn_hair_d_04_r {parent=dyn_hair_d_03_r, loc=<6.9529 0 0>, rot=<0 0 0 1>}
 dyn_hair_e_01_r {parent=dyn_hair_02, loc=<4.5372 6.7216 -4.6591>, rot=<0.7853 -0.2512 0.2918 0.4848>}
 dyn_hair_e_02_r {parent=dyn_hair_e_01_r, loc=<7.5933 0 0>, rot=<0 0 -0.13 -0.9915>}
 dyn_hair_e_03_r {parent=dyn_hair_e_02_r, loc=<6.5574 0 0>, rot=<0.1248 -0.0049 -0.0393 -0.9914>}
 dyn_hair_e_04_r {parent=dyn_hair_e_03_r, loc=<7.8196 0 0>, rot=<0 0 0 1>}
 dyn_hair_c_01aa_r {parent=dyn_hair_02, loc=<3.5005 -0.1167 -4.9986>, rot=<-0.5415 -0.2072 -0.3284 -0.7456>}
 dyn_hair_c_01a_r {parent=dyn_hair_c_01aa_r, loc=<8.4075 0 0>, rot=<0.0379 0.109 -0.3261 -0.9383>}
 dyn_hair_c_01_r {parent=dyn_hair_c_01a_r, loc=<5.4653 0 0>, rot=<-0.0046 -0.0173 -0.2562 -0.9665>}
 dyn_hair_c_02_r {parent=dyn_hair_c_01_r, loc=<5 0 0>, rot=<0 0 -0.0864 -0.9963>}
 dyn_hair_c_01a_m {parent=dyn_hair_02, loc=<7.3298 0.9287 -2.829>, rot=<-0.6132 0.029 -0.3301 -0.7171>}
 dyn_hair_c_01_m {parent=dyn_hair_c_01a_m, loc=<6.5253 0 0>, rot=<-0.0603 0.2004 -0.2971 -0.9317>}
 dyn_hair_c_02_m {parent=dyn_hair_c_01_m, loc=<8.114 0 0>, rot=<0 0 0 1>}
 dyn_hair_e_l {parent=dyn_hair_01, loc=<1.6698 -5.8881 -0.642>, rot=<-0.4884 0.3388 0.4918 -0.6362>}
 dyn_hair_e_01_l {parent=dyn_hair_e_l, loc=<2.4647 -0.0015 -0.0151>, rot=<0.0492 -0.1927 0.0097 -0.98>}
 dyn_front_hair_b_01_m {parent=head, loc=<6.7884 8.7957 -1.2008>, rot=<0.3778 0.516 0.7497 -0.1701>}
 dyn_front_hair_b_02_m {parent=dyn_front_hair_b_01_m, loc=<3.6408 0 0>, rot=<0 0 0 1>}
 dyn_front_hair_a_01_r {parent=head, loc=<4.362 2.6299 7.2504>, rot=<-0.0606 0.2411 0.9629 0.1048>}
 dyn_front_hair_a_02_r {parent=dyn_front_hair_a_01_r, loc=<2.8689 0 0>, rot=<0 0 0 1>}
 dyn_front_hair_a_01_l {parent=head, loc=<7.4357 7.0099 -3.8816>, rot=<0.0267 -0.9386 -0.2383 -0.2482>}
 dyn_front_hair_a_02_l {parent=dyn_front_hair_a_01_l, loc=<2.895 0 0>, rot=<0 0 0 1>}
 dyn_front_hair_b_01_l {parent=head, loc=<8.4598 6.4298 -4.7494>, rot=<0.0336 -0.9523 -0.1199 -0.2786>}
 dyn_front_hair_b_02_l {parent=dyn_front_hair_b_01_l, loc=<3.8507 0 0>, rot=<0 0 0 1>}
 dyn_front_hair_d_01_l {parent=head, loc=<4.7405 3.4802 -6.586>, rot=<-0.0566 -0.9972 0.0031 0.0481>}
 dyn_front_hair_d_02_l {parent=dyn_front_hair_d_01_l, loc=<2.9757 0 0>, rot=<0 0 0 1>}
 dyn_front_hair_c_01_l {parent=head, loc=<4.526 4.9409 -6.2616>, rot=<-0.0888 -0.996 -0.0003 -0.0034>}
 dyn_front_hair_c_02_l {parent=dyn_front_hair_c_01_l, loc=<2.1987 0 0>, rot=<0 0 0 1>}
 base_earring_01_l {parent=head, loc=<-0.6048 -0.5389 -6.0352>, rot=<-0.005 0 1 -0.0063>}
 dyn_earring_01_l {parent=base_earring_01_l, loc=<0 0 0>, rot=<0 0 0 1>}
 base_earring_a_01_r {parent=head, loc=<-0.601 -0.346 5.8133>, rot=<-0.005 -0.0001 0.9996 -0.0266>}
 dyn_earring_a_01_r {parent=base_earring_a_01_r, loc=<0 0 0>, rot=<0 0 0 1>}
 base_earring_01_r {parent=head, loc=<-0.4442 -0.7671 6.191>, rot=<0.0631 0.0057 0.9926 -0.1034>}
 dyn_earring_01_r {parent=base_earring_01_r, loc=<0 0 0>, rot=<0 0 0 1>}
 dyn_front_hair_a_01_m {parent=head, loc=<8.1814 8.6672 -3.5479>, rot=<-0.2153 -0.7974 -0.5537 0.106>}
 dyn_front_hair_a_02_m {parent=dyn_front_hair_a_01_m, loc=<3.9539 0 0>, rot=<-0.0761 0.0018 -0.1302 -0.9886>}
 clavicle_l {parent=spine_05, loc=<5.4831 -0.1053 -3.2946>, rot=<0.0465 -0.7289 -0.009 -0.6829>}
 upperarm_l {parent=clavicle_l, loc=<9.8658 -1.2136 -0.4105>, rot=<0.0164 -0.3245 -0.0269 -0.9454>}
 lowerarm_l {parent=upperarm_l, loc=<26.465 -0.8345 0.259>, rot=<0.0036 0.0161 -0.2298 -0.9731>}
 hand_l {parent=lowerarm_l, loc=<22.7355 -0.3203 -0.6105>, rot=<-0.6784 -0.0223 0.0735 -0.7307>}
 thumb_01_l {parent=hand_l, loc=<1.3118 -0.3191 -2.459>, rot=<0.6546 -0.4042 0.0664 -0.6354>}
 thumb_02_l {parent=thumb_01_l, loc=<5.9336 -0.0448 0.0401>, rot=<-0.0037 0.0213 0.298 -0.9543>}
 thumb_03_l {parent=thumb_02_l, loc=<2.2223 0.3682 -0.0907>, rot=<0.0118 -0.013 0.0066 -0.9998>}
 index_metacarpal_l {parent=hand_l, loc=<1.6535 0.0183 -2.96>, rot=<0.0158 -0.0124 -0.0173 -0.9996>}
 index_01_l {parent=index_metacarpal_l, loc=<8.3008 0.0357 -0.0524>, rot=<-0.0044 0.0485 0.0924 -0.9945>}
 index_02_l {parent=index_01_l, loc=<3.3587 -0.3004 0.0452>, rot=<-0.002 0.0027 0.3225 -0.9466>}
 index_03_l {parent=index_02_l, loc=<2.5578 -0.123 -0.0252>, rot=<-0.002 0.0067 0.2558 -0.9667>}
 middle_metacarpal_l {parent=hand_l, loc=<2.1856 -0.1725 -0.5984>, rot=<0.006 0.0168 -0.0329 -0.9993>}
 middle_01_l {parent=middle_metacarpal_l, loc=<7.3275 -0.2447 0.0129>, rot=<-0.0208 0.0042 0.0858 -0.9961>}
 middle_02_l {parent=middle_01_l, loc=<4.0123 0.1401 0.1175>, rot=<-0.0218 0.0258 0.3512 -0.9357>}
 middle_03_l {parent=middle_02_l, loc=<2.6077 0.0534 0.1373>, rot=<0.0082 0.0188 0.1708 -0.9851>}
 ring_metacarpal_l {parent=hand_l, loc=<2.2074 -0.3625 1.3638>, rot=<-0.0284 0.0722 -0.0458 -0.9959>}
 ring_01_l {parent=ring_metacarpal_l, loc=<7.0295 -0.6197 -0.1446>, rot=<-0.0489 -0.0252 0.1892 -0.9804>}
 ring_02_l {parent=ring_01_l, loc=<3.6174 0.4647 -0.1027>, rot=<0 0 0.2806 -0.9598>}
 ring_03_l {parent=ring_02_l, loc=<2.4152 0.114 0.0199>, rot=<0 0 0.1189 -0.9929>}
 pinky_metacarpal_l {parent=hand_l, loc=<1.7616 -0.9753 3.3814>, rot=<-0.0397 0.0814 -0.069 -0.9935>}
 pinky_01_l {parent=pinky_metacarpal_l, loc=<7.1709 -0.4831 -0.1094>, rot=<-0.0654 -0.0035 0.2242 -0.9723>}
 pinky_02_l {parent=pinky_01_l, loc=<2.8274 0.0893 -0.1043>, rot=<0.0037 0.0011 0.2854 -0.9584>}
 pinky_03_l {parent=pinky_02_l, loc=<2.2046 0.0307 -0.0078>, rot=<-0.0002 0.0098 0.1445 -0.9895>}
 deform_wrist_dn_base_l {parent=hand_l, loc=<0 -1.5326 0>, rot=<0 0 0 -1>}
 deform_wrist_dn_l {parent=deform_wrist_dn_base_l, loc=<0 0 0>, rot=<0 0 0 1>}
 deform_wrist_up_base_l {parent=hand_l, loc=<0 2.3009 0>, rot=<0 0 0 -1>}
 deform_wrist_up_l {parent=deform_wrist_up_base_l, loc=<0 0 0>, rot=<0 0 0 1>}
 hand_attach_l {parent=hand_l, loc=<8.5 -3 0>, rot=<0.7071 0.7071 0 0>}
 weapon_l {parent=hand_l, loc=<-0.1974 -3.4111 0.3811>, rot=<0 0 0 1>}
 lowerarm_twist_02_l {parent=lowerarm_l, loc=<10.6707 0 0>, rot=<0 0 0 1>}
 lowerarm_twist_01_l {parent=lowerarm_l, loc=<18.316 0 0>, rot=<0 0 0 1>}
 deform_elbow_in_base_l {parent=lowerarm_l, loc=<0.1834 3.6062 -0.2572>, rot=<0 0 0 -1>}
 deform_elbow_in_l {parent=deform_elbow_in_base_l, loc=<0 0 0>, rot=<0 0 0 1>}
 deform_elbow_out_base_l {parent=lowerarm_l, loc=<0 -4 0>, rot=<0 0 0 1>}
 deform_elbow_out_l {parent=deform_elbow_out_base_l, loc=<0 0 0>, rot=<0 0 0 1>}
 upperarm_twist_01_l {parent=upperarm_l, loc=<7.2135 -0.1134 0>, rot=<0 0 0 1>}
 deform_upperarm_bk_base_l {parent=upperarm_twist_01_l, loc=<-6.4053 -4.3746 0.314>, rot=<0 0 0 -1>}
 deform_upperarm_bk_l {parent=deform_upperarm_bk_base_l, loc=<0 0 0>, rot=<0 0 0 1>}
 deform_upperarm_fr_base_l {parent=upperarm_twist_01_l, loc=<-5.6312 5.092 0.3354>, rot=<0 0 0 -1>}
 deform_upperarm_fr_l {parent=deform_upperarm_fr_base_l, loc=<0 0 0>, rot=<0 0 0 1>}
 deform_upperarm_up_base_l {parent=upperarm_twist_01_l, loc=<-8.891 0.4825 3.9517>, rot=<0 0 0 -1>}
 deform_upperarm_up_l {parent=deform_upperarm_up_base_l, loc=<0 0 0>, rot=<0 0 0 1>}
 upperarm_twist_02_l {parent=upperarm_l, loc=<19.7998 -0.1134 0>, rot=<0 0 0 1>}
 clavicle_r {parent=spine_05, loc=<5.4831 -0.1053 3.2946>, rot=<0.7289 0.0465 -0.6829 0.009>}
 upperarm_r {parent=clavicle_r, loc=<-9.8658 1.2136 0.4105>, rot=<0.0164 -0.3245 -0.0269 -0.9454>}
 lowerarm_r {parent=upperarm_r, loc=<-26.465 0.8345 -0.259>, rot=<0.0036 0.0161 -0.2298 -0.9731>}
 hand_r {parent=lowerarm_r, loc=<-22.7355 0.3203 0.6106>, rot=<-0.6784 -0.0223 0.0735 -0.7307>}
 thumb_01_r {parent=hand_r, loc=<-1.3118 0.3191 2.459>, rot=<0.6546 -0.4042 0.0664 -0.6354>}
 thumb_02_r {parent=thumb_01_r, loc=<-5.9336 0.0448 -0.0401>, rot=<-0.0037 0.0213 0.298 -0.9543>}
 thumb_03_r {parent=thumb_02_r, loc=<-2.2223 -0.3682 0.0907>, rot=<0.0118 -0.013 0.0066 -0.9998>}
 index_metacarpal_r {parent=hand_r, loc=<-1.6535 -0.0183 2.96>, rot=<0.0158 -0.0124 -0.0173 -0.9996>}
 index_01_r {parent=index_metacarpal_r, loc=<-8.3008 -0.0357 0.0524>, rot=<-0.0044 0.0485 0.0924 -0.9945>}
 index_02_r {parent=index_01_r, loc=<-3.3587 0.3004 -0.0452>, rot=<-0.002 0.0027 0.3225 -0.9466>}
 index_03_r {parent=index_02_r, loc=<-2.5578 0.123 0.0252>, rot=<-0.002 0.0067 0.2558 -0.9667>}
 middle_metacarpal_r {parent=hand_r, loc=<-2.1856 0.1725 0.5984>, rot=<0.006 0.0168 -0.0329 -0.9993>}
 middle_01_r {parent=middle_metacarpal_r, loc=<-7.3275 0.2447 -0.0129>, rot=<-0.0208 0.0042 0.0858 -0.9961>}
 middle_02_r {parent=middle_01_r, loc=<-4.0123 -0.1401 -0.1175>, rot=<-0.0218 0.0258 0.3512 -0.9357>}
 middle_03_r {parent=middle_02_r, loc=<-2.6077 -0.0534 -0.1373>, rot=<0.0082 0.0188 0.1708 -0.9851>}
 ring_metacarpal_r {parent=hand_r, loc=<-2.2074 0.3625 -1.3638>, rot=<-0.0284 0.0722 -0.0458 -0.9959>}
 ring_01_r {parent=ring_metacarpal_r, loc=<-7.0295 0.6197 0.1446>, rot=<-0.0489 -0.0252 0.1892 -0.9804>}
 ring_02_r {parent=ring_01_r, loc=<-3.6174 -0.4647 0.1027>, rot=<0 0 0.2806 -0.9598>}
 ring_03_r {parent=ring_02_r, loc=<-2.4152 -0.114 -0.0199>, rot=<0 0 0.1189 -0.9929>}
 pinky_metacarpal_r {parent=hand_r, loc=<-1.7616 0.9753 -3.3814>, rot=<-0.0397 0.0814 -0.069 -0.9935>}
 pinky_01_r {parent=pinky_metacarpal_r, loc=<-7.1709 0.4831 0.1094>, rot=<-0.0654 -0.0035 0.2242 -0.9723>}
 pinky_02_r {parent=pinky_01_r, loc=<-2.8274 -0.0893 0.1043>, rot=<0.0037 0.0011 0.2854 -0.9584>}
 pinky_03_r {parent=pinky_02_r, loc=<-2.2046 -0.0307 0.0078>, rot=<-0.0002 0.0098 0.1445 -0.9895>}
 deform_wrist_dn_base_r {parent=hand_r, loc=<-0.0001 1.5328 0>, rot=<0 0 0 -1>}
 deform_wrist_dn_r {parent=deform_wrist_dn_base_r, loc=<0 0 0>, rot=<0 0 0 1>}
 deform_wrist_up_base_r {parent=hand_r, loc=<-0.0002 -2.3007 -0.0001>, rot=<0 0 0 -1>}
 deform_wrist_up_r {parent=deform_wrist_up_base_r, loc=<0 0 0>, rot=<0 0 0 1>}
 weapon_r {parent=hand_r, loc=<0.1974 3.4111 -0.3811>, rot=<0 0 0 1>}
 hand_attach_r {parent=hand_r, loc=<-8.5001 2.9999 0>, rot=<0 0 -0.7071 -0.7071>}
 lowerarm_twist_02_r {parent=lowerarm_r, loc=<-10.6707 0 0>, rot=<0 0 0 -1>}
 lowerarm_twist_01_r {parent=lowerarm_r, loc=<-18.316 0 0>, rot=<0 0 0 -1>}
 deform_elbow_in_base_r {parent=lowerarm_r, loc=<-0.1834 -3.6062 0.2572>, rot=<0 0 0 -1>}
 deform_elbow_in_r {parent=deform_elbow_in_base_r, loc=<0 0 0>, rot=<0 0 0 1>}
 deform_elbow_out_base_r {parent=lowerarm_r, loc=<0 4 -0.0001>, rot=<0 0 0 -1>}
 deform_elbow_out_r {parent=deform_elbow_out_base_r, loc=<0 0 0>, rot=<0 0 0 1>}
 upperarm_twist_01_r {parent=upperarm_r, loc=<-7.2135 0.1134 0>, rot=<0 0 0 -1>}
 deform_upperarm_bk_base_r {parent=upperarm_twist_01_r, loc=<6.4053 4.3746 -0.314>, rot=<0 0 0 -1>}
 deform_upperarm_bk_r {parent=deform_upperarm_bk_base_r, loc=<0 0 0>, rot=<0 0 0 1>}
 deform_upperarm_fr_base_r {parent=upperarm_twist_01_r, loc=<5.6312 -5.092 -0.3354>, rot=<0 0 0 -1>}
 deform_upperarm_fr_r {parent=deform_upperarm_fr_base_r, loc=<0 0 0>, rot=<0 0 0 1>}
 deform_upperarm_up_base_r {parent=upperarm_twist_01_r, loc=<8.891 -0.4825 -3.9517>, rot=<0 0 0 -1>}
 deform_upperarm_up_r {parent=deform_upperarm_up_base_r, loc=<0 0 0>, rot=<0 0 0 1>}
 upperarm_twist_02_r {parent=upperarm_r, loc=<-19.7998 0.1134 0>, rot=<0 0 0 -1>}
 attach_backpack {parent=spine_05, loc=<-0.123 -13.473 0>, rot=<-0.0093 -0.707 0.0093 -0.707>}
 Backpack_BR {parent=attach_backpack, type=Null, loc=<0 2.4119 0>, rot=<0 0 0 1>}
 attach_cape {parent=spine_05, loc=<9.017 -11.714 0>, rot=<0.0616 -0.7044 -0.0616 -0.7044>}
 cape {parent=attach_cape, type=Null, loc=<0 2.5 0>, rot=<0 0 0 1>}
 dyn_base_collar_front_L {parent=spine_05, loc=<8.8769 4.0202 -6.165>, rot=<0.9398 0.2154 -0.2649 0.0128>}
 dyn_collar_front_01_L {parent=dyn_base_collar_front_L, loc=<0 0 0>, rot=<0 0 0 1>}
 dyn_collar_front_02_L {parent=dyn_collar_front_01_L, loc=<5.4992 0 0>, rot=<0 0 0 1>}
 dyn_base_collar_side_L {parent=spine_05, loc=<11.0644 -2.6685 -4.5696>, rot=<0.9722 0.0911 -0.1748 0.1263>}
 dyn_collar_side_01_L {parent=dyn_base_collar_side_L, loc=<0 0 0>, rot=<0 0 0 1>}
 dyn_collar_side_02_L {parent=dyn_collar_side_01_L, loc=<2.6527 0 0>, rot=<0 0 0 1>}
 dyn_base_collar_front_R {parent=spine_05, loc=<8.4908 4.2129 6.2139>, rot=<-0.0768 0.2524 -0.1714 -0.9492>}
 dyn_collar_front_01_R {parent=dyn_base_collar_front_R, loc=<0 0 0>, rot=<0 0 0 1>}
 dyn_collar_front_02_R {parent=dyn_collar_front_01_R, loc=<5.6175 0 0.1313>, rot=<0 0.0614 0 -0.9981>}
 dyn_base_collar_side_R {parent=spine_05, loc=<10.8046 -0.982 5.6598>, rot=<-0.1285 0.1075 0.0187 -0.9857>}
 dyn_collar_side_01_R {parent=dyn_base_collar_side_R, loc=<0 0 0>, rot=<0 0 0 1>}
 dyn_collar_side_02_R {parent=dyn_collar_side_01_R, loc=<2.7519 0 0>, rot=<0 0 0 1>}
 dyn_base_collar_Back {parent=spine_05, loc=<11.369 -4.5876 -0.001>, rot=<-0.7061 0.0381 0.0381 -0.7061>}
 dyn_collar_Back_01 {parent=dyn_base_collar_Back, loc=<0 0 0>, rot=<0 0 0 1>}
 dyn_collar_Back_02 {parent=dyn_collar_Back_01, loc=<2.0074 0 0>, rot=<0 0 0 1>}
 deform_pec_aim_base_l {parent=spine_04, loc=<3.2456 9.3309 -1.2237>, rot=<0.0603 -0.4033 0.3318 -0.8507>}
 deform_pec_aim_l {parent=deform_pec_aim_base_l, loc=<0 0 0>, rot=<0 0 0 1>}
 deform_pec_base_l {parent=deform_pec_aim_l, loc=<9.3665 0 0>, rot=<0 0 0 1>}
 deform_pec_l {parent=deform_pec_base_l, loc=<0 0 0>, rot=<0 0 0 1>}
 deform_pec_aim_base_r {parent=spine_04, loc=<3.2452 9.3309 1.2237>, rot=<-0.4033 -0.0603 0.8507 0.3318>}
 deform_pec_aim_r {parent=deform_pec_aim_base_r, loc=<0 0 0>, rot=<0 0 0 1>}
 deform_pec_base_r {parent=deform_pec_aim_r, loc=<-9.3667 -0.0003 -0.0004>, rot=<0 0 0 1>}
 deform_pec_r {parent=deform_pec_base_r, loc=<0 0 0>, rot=<0 0 0 1>}
 dyn_base_jacket_side_R {parent=spine_04, loc=<0.9723 2.4154 12.1775>, rot=<-0.0353 0.122 0.9913 0.0353>}
 dyn_jacket_side_01_R {parent=dyn_base_jacket_side_R, loc=<0 0 0>, rot=<0 0 0 1>}
 dyn_jacket_side_02_R {parent=dyn_jacket_side_01_R, loc=<6.9658 0 -0.0877>, rot=<0 0 0 1>}
 dyn_base_jacket_back_R {parent=spine_04, loc=<1.6798 -8.4797 6.703>, rot=<-0.0131 -0.5403 0.8393 0.0589>}
 dyn_jacket_back_01_R {parent=dyn_base_jacket_back_R, loc=<0 0 0>, rot=<0 0 0 1>}
 dyn_jacket_back_02_R {parent=dyn_jacket_back_01_R, loc=<7.5503 0 -0.0877>, rot=<0 0 0 1>}
 dyn_base_jacket_back_L {parent=spine_04, loc=<1.6407 -8.716 -6.7947>, rot=<-0.0831 -0.829 0.551 0.048>}
 dyn_jacket_back_01_L {parent=dyn_base_jacket_back_L, loc=<0 0 0>, rot=<0 0 0 1>}
 dyn_jacket_back_02_L {parent=dyn_jacket_back_01_L, loc=<7.2077 0 0>, rot=<0 0 0 1>}
 dyn_base_jacket_front_L {parent=spine_04, loc=<1.697 12.4397 -7.5745>, rot=<0.021 -0.9074 -0.4191 -0.0229>}
 dyn_jacket_front_L_01 {parent=dyn_base_jacket_front_L, loc=<0 0 0>, rot=<0 0 0 1>}
 dyn_jacket_front_L_02 {parent=dyn_jacket_front_L_01, loc=<8.5934 0 0>, rot=<0 0 0 1>}
 dyn_base_jacket_side_L {parent=spine_04, loc=<1.2696 2.5614 -12.9155>, rot=<-0.0066 -0.9873 -0.1305 0.0905>}
 dyn_jacket_side_01_L {parent=dyn_base_jacket_side_L, loc=<0 0 0>, rot=<0 0 0 1>}
 dyn_jacket_side_02_L {parent=dyn_jacket_side_01_L, loc=<7.0502 0 0>, rot=<0 0 0 1>}
 dyn_base_jacket_front_R {parent=spine_04, loc=<1.7735 12.9807 7.0716>, rot=<0.0308 0.5978 0.8001 -0.0381>}
 dyn_jacket_front_R_01 {parent=dyn_base_jacket_front_R, loc=<0 0 0>, rot=<0 0 0 1>}
 dyn_jacket_front_R_02 {parent=dyn_jacket_front_R_01, loc=<8.4239 0 0>, rot=<0 0 0 1>}
 thigh_l {parent=pelvis, loc=<-1.0855 -0.2109 -8.7053>, rot=<0.0009 0.0355 0.0155 -0.9993>}
 calf_l {parent=thigh_l, loc=<-42.5284 -0.0019 -0.7297>, rot=<0.0006 -0.0161 -0.0391 -0.9991>}
 foot_l {parent=calf_l, loc=<-46.7401 -2.5014 -0.592>, rot=<0.0427 -0.0238 0.0353 -0.9982>}
 ball_l {parent=foot_l, loc=<-3.5658 10.8066 -0.1544>, rot=<0.0001 -0.0001 -0.7063 -0.7079>}
 calf_twist_02_l {parent=calf_l, loc=<-19.1546 -1.7943 -0.2638>, rot=<0.0237 0.0095 -0.0035 -0.9997>}
 calf_twist_01_l {parent=calf_l, loc=<-35.8483 -1.9197 -0.5794>, rot=<0.0237 0.0095 -0.0035 -0.9997>}
 deform_knee_in_base_l {parent=calf_l, loc=<0.8031 -5.0231 0.2456>, rot=<0 0 0 -1>}
 deform_knee_in_l {parent=deform_knee_in_base_l, loc=<0 0 0>, rot=<0 0 0 1>}
 deform_knee_out_base_l {parent=calf_l, loc=<0 4 0>, rot=<0 0 0 -1>}
 deform_knee_out_l {parent=deform_knee_out_base_l, loc=<0 0 0>, rot=<0 0 0 1>}
 thigh_twist_01_l {parent=thigh_l, loc=<-27.952 0 0>, rot=<0.0013 0 0 -1>}
 deform_groin_base_l {parent=thigh_l, loc=<0.4801 5.6776 -0.2486>, rot=<0 0 0 -1>}
 deform_groin_l {parent=deform_groin_base_l, loc=<0 0 0>, rot=<0 0 0 1>}
 thigh_r {parent=pelvis, loc=<-1.0855 -0.2109 8.7053>, rot=<0.0355 -0.0009 0.9993 0.0155>}
 calf_r {parent=thigh_r, loc=<42.5284 0.0019 0.7297>, rot=<0.0006 -0.0161 -0.0391 -0.9991>}
 foot_r {parent=calf_r, loc=<46.7401 2.5014 0.592>, rot=<0.0427 -0.0238 0.0353 -0.9982>}
 ball_r {parent=foot_r, loc=<3.5658 -10.8066 0.1544>, rot=<0.0001 -0.0001 -0.7063 -0.7079>}
 calf_twist_02_r {parent=calf_r, loc=<19.1546 1.7943 0.2638>, rot=<0.0237 0.0095 -0.0035 -0.9997>}
 calf_twist_01_r {parent=calf_r, loc=<35.8483 1.9197 0.5794>, rot=<0.0237 0.0095 -0.0035 -0.9997>}
 deform_knee_in_base_r {parent=calf_r, loc=<-0.8031 5.0231 -0.2457>, rot=<0 0 0 -1>}
 deform_knee_in_r {parent=deform_knee_in_base_r, loc=<0 0 0>, rot=<0 0 0 1>}
 deform_knee_out_base_r {parent=calf_r, loc=<0 -4 0>, rot=<0 0 0 -1>}
 deform_knee_out_r {parent=deform_knee_out_base_r, loc=<0 0 0>, rot=<0 0 0 1>}
 thigh_twist_01_r {parent=thigh_r, loc=<27.952 0 0>, rot=<0.0013 0 0 -1>}
 deform_groin_base_r {parent=thigh_r, loc=<-0.4801 -5.6776 0.2486>, rot=<0 0 0 -1>}
 deform_groin_r {parent=deform_groin_base_r, loc=<0 0 0>, rot=<0 0 0 1>}
 deform_glute_base_l {parent=pelvis, loc=<-7.3466 -9.0213 -9.1245>, rot=<0.0172 0.031 -0.4477 -0.8935>}
 deform_glute_l {parent=deform_glute_base_l, loc=<0 0 0>, rot=<0 0 0 1>}
 deform_glute_base_r {parent=pelvis, loc=<-7.3464 -9.0213 9.1245>, rot=<0.031 -0.0172 0.8935 -0.4477>}
 deform_glute_r {parent=deform_glute_base_r, loc=<0 0 0>, rot=<0 0 0 1>}
 dyn_Belt_MSTR {parent=pelvis, loc=<1.7442 -3.2336 -13.5646>, rot=<0.5866 0.03 0.8082 0.043>}
 dyn_Belt_MSTR_01 {parent=dyn_Belt_MSTR, loc=<0 0 0>, rot=<0 0 0 1>}
 dyn_base_Belt_l {parent=dyn_Belt_MSTR_01, loc=<33.5181 1.5475 2.3444>, rot=<0.0808 0.0244 -0.0121 -0.9964>}
 dyn_Belt_l_01 {parent=dyn_base_Belt_l, loc=<0 0 0>, rot=<0 0 0 1>}
 dyn_base_Belt_Mid {parent=dyn_Belt_MSTR_01, loc=<19.4484 -12.119 1.2564>, rot=<0.038 0.0198 -0.0704 -0.9966>}
 dyn_Belt_Mid_01 {parent=dyn_base_Belt_Mid, loc=<0 0 0>, rot=<0 0 0 1>}
 dyn_base_Belt_Mid_02 {parent=dyn_Belt_MSTR_01, loc=<28.8231 -9.0932 1.8554>, rot=<0.1143 -0.0148 -0.3512 -0.9292>}
 dyn_Belt_Mid_02 {parent=dyn_base_Belt_Mid_02, loc=<0 0 0>, rot=<0 0 0 1>}
 dyn_base_Belt_Back {parent=dyn_Belt_MSTR, loc=<4.2627 11.1821 -0.5949>, rot=<-0.0087 -0.0008 -0.1225 -0.9924>}
 dyn_Belt_Back_01 {parent=dyn_base_Belt_Back, loc=<0 0 0>, rot=<0 0 0 1>}
 dyn_base_Belt_Front {parent=dyn_Belt_MSTR, loc=<7.4082 -11.2525 0.6979>, rot=<0.0811 0.0008 -0.0102 -0.9967>}
 dyn_Belt_Front_01 {parent=dyn_base_Belt_Front, loc=<0 0 0>, rot=<0 0 0 1>}
 dyn_base_Belt_Front_L {parent=dyn_Belt_MSTR, loc=<-0.8576 -7.0465 0.0101>, rot=<-0.0096 -0.0734 0.2237 -0.9718>}
 dyn_Belt_Front_01_l {parent=dyn_base_Belt_Front_L, loc=<0 0 0>, rot=<0 0 0 1>}
 attach {parent=root, loc=<0 0 -2.1727>, rot=<0 0 -0.0368 -0.9993>}
 ik_hand_root {parent=root, loc=<0 0 0>, rot=<0 0 0 1>}
 ik_hand_gun {parent=ik_hand_root, loc=<-48.1743 3.3545 104.4381>, rot=<0.6276 0.1309 0.4183 -0.6434>}
 ik_hand_r {parent=ik_hand_gun, loc=<0 0 0>, rot=<0 0 0 1>}
 ik_hand_l {parent=ik_hand_gun, loc=<59.3271 67.6997 34.3524>, rot=<-0.6158 -0.7027 -0.3565 0>}
 ik_foot_root {parent=root, loc=<0 0 0>, rot=<0 0 0 1>}
 ik_foot_l {parent=ik_foot_root, loc=<39.5635 26.5333 10.3398>, rot=<0.1189 0.7081 0.403 -0.5675>}
 ik_foot_r {parent=ik_foot_root, loc=<-21.0797 28.16 9.7927>, rot=<-0.6 -0.3118 0.7334 -0.0701>}
 root_fp {parent=root, loc=<0 0 0>, rot=<0 0 0 1>}
 attach_fp {parent=root_fp, loc=<0 0 0>, rot=<0 0 0 1>}
 camera_root {parent=root_fp, loc=<0 0 0>, rot=<0 0 0 1>}
 camera {parent=camera_root, loc=<0 0 0>, rot=<0 0 0 1>}
 attach_emote_01 {parent=root, loc=<0 0 0>, rot=<0 0 0 1>}
 attach_emote_02 {parent=root, loc=<0 0 0>, rot=<0 0 0 1>}
 attach_emote_03 {parent=root, loc=<0 0 0>, rot=<0 0 0 1>}
 warp_point_01 {parent=root, loc=<0 0 0>, rot=<0 0 0 1>}
 warp_point_02 {parent=root, loc=<0 0 0>, rot=<0 0 0 1>}
 warp_point_03 {parent=root, loc=<0 0 0>, rot=<0 0 0 1>}
]
variables [SkyDive_Back_ADJ {type=double, default=0}]
