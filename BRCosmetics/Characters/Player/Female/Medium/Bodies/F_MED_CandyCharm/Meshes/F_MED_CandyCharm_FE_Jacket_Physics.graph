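# Physics asset for the CandyCharm female medium body's FE jacket mesh, providing secondary motion for jacket panels and belt pieces.
# Capsules on pelvis, spine_04 and right thigh act as body shapes, while front, side and back jacket panels hang from tiny base-bone capsules through angle-limited constraints.
# Most jacket bases link to the pelvis, except the left side, which is locked to spine_04.
# A multi-capsule master belt body connects to the mid and left belt pieces with tight limits.
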PhysicsAsset /BRCosmetics/Characters/Player/Female/Medium/Bodies/F_MED_CandyCharm/Meshes/F_MED_CandyCharm_FE_Jacket_Physics
bodies {
 dyn_Belt_Back_01 [{
   shapeType=Capsule
   length=3.713
   radius=2.2969
   shapeName=dyn_Belt_Back_01_capsule
   center=<-0.0845 1.458 -0.0137>
   rotation=<rot 3.9833 -64.4716 81.7245>
  }]
 dyn_Belt_MSTR [{shapeType=Capsule, length=0.505, radius=0.505, shapeName=dyn_Belt_MSTR_capsule, rotation=<rot 0 -90 0>}]
 dyn_Belt_MSTR_01 [
  {
   shapeType=Capsule
   length=7.5525
   radius=3.0013
   shapeName=dyn_Belt_MSTR_01_capsule
   center=<37.4076 1.674 2.448>
   rotation=<rot -62.8994 -0.0001 -90.2867>
  }
  {
   shapeType=Capsule
   length=16.982
   radius=3.2734
   shapeName=dyn_Belt_MSTR_01_capsule
   center=<28.0181 15.4295 0.9883>
   rotation=<rot -80.1168 5.9595 -44.6727>
  }
  {
   shapeType=Capsule
   length=10.3567
   radius=3.4395
   shapeName=dyn_Belt_MSTR_01_capsule
   center=<35.6594 -5.3006 2.1516>
   rotation=<rot 3.8381 -36.4615 -93.1411>
  }
  {
   shapeType=Capsule
   length=13.8491
   radius=2.6862
   shapeName=dyn_Belt_MSTR_01_capsule
   center=<14.3211 16.9613 -0.4119>
   rotation=<rot -77.8876 -137.2781 156.3894>
  }
  {
   shapeType=Capsule
   length=12.625
   radius=4.125
   shapeName=dyn_Belt_MSTR_01_capsule
   center=<20.117 14.3913 -0.1497>
   rotation=<rot -85 -0.0002 0.0004>
  }
 ]
 dyn_Belt_Mid_01 [{shapeType=Capsule, length=6, radius=3, shapeName=dyn_Belt_Mid_01_capsule, center=<-0.0151 0.0854 0>, rotation=<rot 0.0084 -80.0024 -84.4281>}]
 dyn_Belt_Mid_02 [{
   shapeType=Capsule
   length=7.125
   radius=4.875
   shapeName=dyn_Belt_Mid_02_capsule
   center=<-0.2743 -0.86 0.6379>
   rotation=<rot -3.6744 -100.1628 80.3269>
  }]
 dyn_Belt_l_01 [{
   shapeType=Capsule
   length=10
   radius=3
   shapeName=dyn_Belt_l_01_capsule
   center=<2.3369 2.3633 -0.2386>
   rotation=<rot -18.9919 179.5694 -80.2305>
  }]
 dyn_base_Belt_Back [{shapeType=Capsule, length=0.505, radius=0.505, shapeName=dyn_base_Belt_Back_capsule, rotation=<rot 0 -90 0>}]
 dyn_base_jacket_back_L [{shapeType=Capsule, length=0.505, radius=0.505, shapeName=dyn_base_jacket_back_L_capsule, rotation=<rot 0 -90 0>}]
 dyn_base_jacket_back_R [{shapeType=Capsule, length=0.505, radius=0.505, shapeName=dyn_base_jacket_back_R_capsule, rotation=<rot 0 -90 0>}]
 dyn_base_jacket_front_L [{shapeType=Capsule, length=0.505, radius=0.505, shapeName=dyn_base_jacket_front_L_capsule, rotation=<rot 0 -90 0>}]
 dyn_base_jacket_front_R [{shapeType=Capsule, length=0.505, radius=0.505, shapeName=dyn_base_jacket_front_R_capsule, rotation=<rot 0 -90 0>}]
 dyn_base_jacket_side_L [{shapeType=Capsule, length=0.505, radius=0.505, shapeName=dyn_base_jacket_side_L_capsule, rotation=<rot 0 -90 0>}]
 dyn_base_jacket_side_R [{shapeType=Capsule, length=0.505, radius=0.505, shapeName=dyn_base_jacket_side_R_capsule, rotation=<rot 0 -90 0>}]
 dyn_jacket_back_01_L [{
   shapeType=Capsule
   length=8.5
   radius=2.625
   shapeName=dyn_jacket_back_01_L_capsule
   center=<1.5432 0.1247 1.44>
   rotation=<rot 3.9165 -101.5395 -98.596>
  }]
 dyn_jacket_back_01_R [{
   shapeType=Capsule
   length=7.875
   radius=2.8302
   shapeName=dyn_jacket_back_01_R_capsule
   center=<1.9052 -0.0903 1.4863>
   rotation=<rot -8.1297 -91.5692 -95.7106>
  }]
 dyn_jacket_front_L_01 [{
   shapeType=Capsule
   length=10
   radius=2.7332
   shapeName=dyn_jacket_front_L_01_capsule
   center=<2.3096 -0.5182 2.148>
   rotation=<rot 75.1633 -84.3927 95.3704>
  }]
 dyn_jacket_front_R_01 [{
   shapeType=Capsule
   length=10
   radius=2.5168
   shapeName=dyn_jacket_front_L_01_capsule
   center=<3.4338 0.1818 1.5634>
   rotation=<rot 67.9873 80.2654 82.5989>
  }]
 dyn_jacket_side_01_L [{
   shapeType=Capsule
   length=7.0679
   radius=2.8934
   shapeName=dyn_jacket_side_01_L_capsule
   center=<3.2757 -0.2862 2.7471>
   rotation=<rot 10.6948 92.3556 -79.5562>
  }]
 dyn_jacket_side_01_R [{
   shapeType=Capsule
   length=10
   radius=2.8934
   shapeName=dyn_jacket_side_01_L_capsule
   center=<3.6442 -0.3313 3.0659>
   rotation=<rot 10.6948 92.3555 -86.6955>
  }]
 pelvis [{shapeType=Capsule, length=14.8113, radius=11.2417, shapeName=pelvis_capsule, center=<-5.6005 -2.1275 0>, rotation=<rot 6.1186 -0.635 0.1501>}]
 spine_04 [{
   shapeType=Capsule
   length=13.8532
   radius=11.4762
   shapeName=spine_04_capsule
   center=<-1.3249 2.6424 -0.131>
   rotation=<rot -1.4872 -90.0105 89.9717>
  }]
 thigh_r [{
   shapeType=Capsule
   length=28.4611
   radius=10.9374
   shapeName=thigh_r_capsule
   center=<16.6531 0.5669 -0.7141>
   rotation=<rot -88.7831 -0.0022 -6.1513>
  }]
}
constraints [
 {bone1Name=dyn_jacket_front_L_01, bone2Name=dyn_base_jacket_front_L, swing1LimitDegrees=20, swing2LimitDegrees=45, twistLimitDegrees=0}
 {
  bone1Name=dyn_base_jacket_side_L
  bone2Name=spine_04
  swing1LimitDegrees=0
  swing1Motion=Locked
  swing2LimitDegrees=45
  swing2Motion=Locked
  twistLimitDegrees=45
  twistMotion=Locked
 }
 {bone1Name=dyn_jacket_front_R_01, bone2Name=dyn_base_jacket_front_R, swing1LimitDegrees=20, swing2LimitDegrees=45, twistLimitDegrees=0}
 {bone1Name=dyn_jacket_side_01_L, bone2Name=dyn_base_jacket_side_L, swing1LimitDegrees=1, swing2LimitDegrees=1, twistLimitDegrees=0}
 {bone1Name=dyn_jacket_side_01_R, bone2Name=dyn_base_jacket_side_R, swing1LimitDegrees=1, swing2LimitDegrees=1, twistLimitDegrees=0}
 {bone1Name=dyn_jacket_back_01_L, bone2Name=dyn_base_jacket_back_L, swing1LimitDegrees=5, swing2LimitDegrees=5, twistLimitDegrees=0}
 {bone1Name=dyn_jacket_back_01_R, bone2Name=dyn_base_jacket_back_R, swing1LimitDegrees=5, swing2LimitDegrees=5, twistLimitDegrees=0}
 {
  bone1Name=spine_04
  bone2Name=pelvis
  swing1LimitDegrees=45
  swing1Motion=Free
  swing2LimitDegrees=45
  swing2Motion=Free
  twistLimitDegrees=45
  twistMotion=Free
 }
 {
  bone1Name=dyn_base_jacket_side_R
  bone2Name=pelvis
  swing1LimitDegrees=45
  swing1Motion=Free
  swing2LimitDegrees=45
  swing2Motion=Free
  twistLimitDegrees=45
  twistMotion=Free
 }
 {bone1Name=dyn_base_jacket_back_R, bone2Name=pelvis, swing1LimitDegrees=1, swing2LimitDegrees=5, twistLimitDegrees=0}
 {
  bone1Name=dyn_base_jacket_back_L
  bone2Name=pelvis
  swing1LimitDegrees=1
  swing1Motion=Locked
  swing2LimitDegrees=5
  swing2Motion=Locked
  twistLimitDegrees=0
  twistMotion=Locked
 }
 {
  bone1Name=dyn_base_jacket_front_L
  bone2Name=pelvis
  swing1LimitDegrees=45
  swing1Motion=Free
  swing2LimitDegrees=45
  swing2Motion=Free
  twistLimitDegrees=45
  twistMotion=Free
 }
 {
  bone1Name=dyn_base_jacket_front_R
  bone2Name=pelvis
  swing1LimitDegrees=45
  swing1Motion=Free
  swing2LimitDegrees=45
  swing2Motion=Free
  twistLimitDegrees=45
  twistMotion=Free
 }
 {bone1Name=dyn_Belt_MSTR_01, bone2Name=dyn_Belt_MSTR, swing1LimitDegrees=2, swing2LimitDegrees=2, twistLimitDegrees=0}
 {bone1Name=dyn_Belt_Back_01, bone2Name=dyn_base_Belt_Back, swing1LimitDegrees=5, swing2LimitDegrees=5, twistLimitDegrees=0}
 {bone1Name=thigh_r, bone2Name=pelvis, swing1LimitDegrees=20, swing2LimitDegrees=20, twistLimitDegrees=0}
 {bone1Name=dyn_Belt_MSTR_01, bone2Name=dyn_Belt_Mid_01, swing1LimitDegrees=2, swing2LimitDegrees=2, twistLimitDegrees=0}
 {bone1Name=dyn_Belt_MSTR_01, bone2Name=dyn_Belt_Mid_02, swing1LimitDegrees=2, swing2LimitDegrees=2, twistLimitDegrees=0}
 {bone1Name=dyn_Belt_MSTR_01, bone2Name=dyn_Belt_l_01, swing1LimitDegrees=2, swing2LimitDegrees=2, twistLimitDegrees=0}
]
