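# Character-part Animation Blueprint for the F_MED_CandyCharm body, adding secondary motion to its belt, jacket and collar.
# It copies the parent mesh pose through a Control Rig with a skydive back adjustment, then chains RigidBody simulations, limit-constrained AnimDynamics and additive bone offsets.
# The update event turns movement, crouch and skydive states into float alphas, while emote state directly toggles jacket offsets.
# Computed GravityOverride, SkyDive and FE_DYN values are unused in the shown AnimGraph.
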AnimBlueprint /BRCosmetics/Characters/Player/Female/Medium/Bodies/F_MED_CandyCharm/Meshes/F_MED_CandyCharm_AnimBP
target_skeleton=&/F_MED_CandyCharm_Skeleton
graphs [
 AnimGraph {nodes [
   n1 = Get Crouch() -> (Crouch:double)
   "Transform (Modify) Bone - Bone: dyn_base_Belt_l"(Alpha=@n1, Translation="0,-2,0", Rotation="0,0,0", Scale="1,1,1") {
    rotationMode=BMM_Additive
    rotationSpace=BCS_BoneSpace
    translationMode=BMM_Additive
    translationSpace=BCS_BoneSpace
    alphaBoolBlend {blendOption=Linear}
    boneToModify {boneName=dyn_base_Belt_l}
    translation {y=-2}
   }
   n2 = Get Moving() -> (Moving:double)
   n3 = Control Rig!(Source=Copy Pose From Mesh() {bUseAttachedParent=true}, Alpha, SkyDive_Back_ADJ:double=Get SkyDive_Back_ADJ() -> (SkyDive_Back_ADJ:double)) -> (Pose) {
    alphaBoolBlend {blendOption=Linear}
    controlRigAssetReference {blueprintRigClass=&/F_MED_CandyCharm_CtrlRig.F_MED_CandyCharm_CtrlRig_C}
   }
   n4 = "Transform (Modify) Bone - Bone: dyn_Belt_MSTR_01"(ComponentPose=Local To Component(LocalPose=@n3), Alpha=1, Translation="0,0,0", Rotation="-5,0,0", Scale="1,1,1") {
    rotationMode=BMM_Additive
    translationMode=BMM_Additive
    alphaBoolBlend {blendOption=Linear}
    boneToModify {boneName=dyn_Belt_MSTR_01}
    rotation {pitch=-5}
   }
   n5 = RigidBody!(ComponentPose=@n4, Alpha, OverrideWorldGravity:Vector="0,0,-980", ExternalForce) -> (Pose) {
    bOverrideWorldGravity=true
    overridePhysicsAsset=&/F_MED_CandyCharm_Jacket_Physics
    simulationSpace=BaseBoneSpace
    alphaBoolBlend {blendOption=Linear}
    baseBoneRef {boneName=spine_02}
    simSpaceSettings {maxAngularAcceleration=100, maxAngularVelocity=10, maxLinearAcceleration=2000, maxLinearVelocity=450, worldAlpha=1}
   }
   n6 = "Transform (Modify) Bone - Bone: dyn_base_Belt_Front_L"(ComponentPose=@n5, Alpha=1, Translation="0,-5,0", Rotation="0,-10,0", Scale="1,1,1") {
    rotationMode=BMM_Additive
    rotationSpace=BCS_BoneSpace
    translationMode=BMM_Additive
    translationSpace=BCS_BoneSpace
    alphaBoolBlend {blendOption=Linear}
    boneToModify {boneName=dyn_base_Belt_Front_L}
    rotation {yaw=-10}
    translation {y=-5}
   }
   n7 = "Transform (Modify) Bone - Bone: dyn_base_Belt_Front"(ComponentPose=@n6, Alpha=@n2, Translation="0,-3,0", Rotation="0,0,0", Scale="1,1,1") {
    rotationMode=BMM_Additive
    rotationSpace=BCS_BoneSpace
    translationMode=BMM_Additive
    translationSpace=BCS_BoneSpace
    alphaBoolBlend {blendOption=Linear}
    boneToModify {boneName=dyn_base_Belt_Front}
    translation {y=-3}
   }
   n8 = "Transform (Modify) Bone - Bone: dyn_base_Belt_Front_L"(ComponentPose=@n7, Alpha=@n2, Translation="0,0,0", Rotation="0,0,0", Scale="1,1,1") {
    rotationMode=BMM_Additive
    rotationSpace=BCS_BoneSpace
    translationMode=BMM_Additive
    translationSpace=BCS_BoneSpace
    alphaBoolBlend {blendOption=Linear}
    boneToModify {boneName=dyn_base_Belt_Front_L}
   }
   n9 = "Transform (Modify) Bone - Bone: dyn_base_Belt_Mid"(ComponentPose=@n8, Alpha=@n2, Translation="0,-3,0", Rotation="0,0,0", Scale="1,1,1") {
    rotationMode=BMM_Additive
    rotationSpace=BCS_BoneSpace
    translationMode=BMM_Additive
    translationSpace=BCS_BoneSpace
    alphaBoolBlend {blendOption=Linear}
    boneToModify {boneName=dyn_base_Belt_Mid}
    translation {y=-3}
   }
   n10 = Anim Dynamics - Bone: dyn_collar_front_01_L(ComponentPose=@n9, Alpha=0.5, ExternalForce="0,0,0") {
    angularDampingOverride=0.9999
    angularSpringConstant=20000
    bAngularSpring=true
    bOverrideAngularDamping=true
    bUseGravityOverride=true
    bUseSphericalLimits=true
    simulationSpace=RootRelative
    alphaBoolBlend {blendOption=Linear}
    boundBone {boneName=dyn_collar_front_01_L}
    componentAppliedLinearAccClamp=<2000 2000 2000>
    componentLinearAccScale=<0.2 0.2 0.2>
    componentLinearVelScale=<0.2 0.2 0.2>
    gravityOverride {z=-980}
    physicsBodyDefinitions [{
      boundBone {boneName=dyn_collar_front_01_L}
      boxExtents=<20 30 30>
      constraintSetup {angularLimitsMax=<0 5 5>, angularLimitsMin=<0 -5 -5>, angularTarget=<1 0 0>}
      localJointOffset=<3 0 0>
     }]
    sphericalLimits [
     {limitRadius=11, drivingBone {boneName=head}, sphereLocalOffset=<0 2 0>}
     {limitRadius=10, drivingBone {boneName=clavicle_l}, sphereLocalOffset=<10 0 0>}
    ]
   }
   n11 = Anim Dynamics - Bone: dyn_collar_front_02_L(ComponentPose=@n10, Alpha=0.5, ExternalForce="0,0,0") {
    angularDampingOverride=0.9999
    angularSpringConstant=20000
    bAngularSpring=true
    bOverrideAngularDamping=true
    bUseGravityOverride=true
    bUseSphericalLimits=true
    simulationSpace=RootRelative
    alphaBoolBlend {blendOption=Linear}
    boundBone {boneName=dyn_collar_front_02_L}
    componentAppliedLinearAccClamp=<2000 2000 2000>
    componentLinearAccScale=<0.2 0.2 0.2>
    componentLinearVelScale=<0.2 0.2 0.2>
    gravityOverride {z=-980}
    physicsBodyDefinitions [{
      boundBone {boneName=dyn_collar_front_02_L}
      boxExtents=<20 30 30>
      constraintSetup {angularLimitsMax=<0 5 5>, angularLimitsMin=<0 -5 -5>, angularTarget=<1 0 0>}
      localJointOffset=<3 0 0>
     }]
    sphericalLimits [
     {limitRadius=12, drivingBone {boneName=head}, sphereLocalOffset=<-2 2 0>}
     {limitRadius=5, drivingBone {boneName=upperarm_l}, sphereLocalOffset=<5 0 0>}
    ]
   }
   n12 = Anim Dynamics - Bone: dyn_collar_side_01_L(ComponentPose=@n11, Alpha=0.5, ExternalForce="0,0,0") {
    angularDampingOverride=0.9999
    angularSpringConstant=20000
    bAngularSpring=true
    bOverrideAngularDamping=true
    bUseGravityOverride=true
    bUseSphericalLimits=true
    simulationSpace=RootRelative
    alphaBoolBlend {blendOption=Linear}
    boundBone {boneName=dyn_collar_side_01_L}
    componentAppliedLinearAccClamp=<2000 2000 2000>
    componentLinearAccScale=<0.2 0.2 0.2>
    componentLinearVelScale=<0.2 0.2 0.2>
    gravityOverride {z=-980}
    physicsBodyDefinitions [{
      boundBone {boneName=dyn_collar_side_01_L}
      boxExtents=<20 30 30>
      constraintSetup {angularLimitsMax=<0 5 5>, angularLimitsMin=<0 -5 -5>, angularTarget=<1 0 0>}
      localJointOffset=<3 0 0>
     }]
    sphericalLimits [
     {limitRadius=10, drivingBone {boneName=head}, sphereLocalOffset=<-5 2 0>}
     {limitRadius=8, drivingBone {boneName=clavicle_l}, sphereLocalOffset=<8 0 -1>}
    ]
   }
   n13 = Anim Dynamics - Bone: dyn_collar_front_01_R(ComponentPose=@n12, Alpha=0.5, ExternalForce="0,0,0") {
    angularDampingOverride=0.9999
    angularSpringConstant=20000
    bAngularSpring=true
    bOverrideAngularDamping=true
    bUseGravityOverride=true
    simulationSpace=RootRelative
    alphaBoolBlend {blendOption=Linear}
    boundBone {boneName=dyn_collar_front_01_R}
    componentAppliedLinearAccClamp=<2000 2000 2000>
    componentLinearAccScale=<0.2 0.2 0.2>
    componentLinearVelScale=<0.2 0.2 0.2>
    gravityOverride {z=-980}
    physicsBodyDefinitions [{
      boundBone {boneName=dyn_collar_front_01_R}
      boxExtents=<20 30 30>
      constraintSetup {angularLimitsMax=<0 5 5>, angularLimitsMin=<0 -5 -5>, angularTarget=<1 0 0>}
      localJointOffset=<3 0 0>
     }]
    planarLimits [{drivingBone {boneName=spine_05}, planeTransform {location=<7.9263 -0.128 7.2906>, rotation=<rot 0 0 -9.9521>}}]
    sphericalLimits [{limitRadius=9, drivingBone {boneName=neck_01}, sphereLocalOffset=<0 2 0>}]
   }
   n14 = Anim Dynamics - Bone: dyn_collar_front_02_R(ComponentPose=@n13, Alpha=0.5, ExternalForce="0,0,0") {
    angularDampingOverride=0.9999
    angularSpringConstant=20000
    bAngularSpring=true
    bOverrideAngularDamping=true
    bUseGravityOverride=true
    bUseSphericalLimits=true
    simulationSpace=RootRelative
    alphaBoolBlend {blendOption=Linear}
    boundBone {boneName=dyn_collar_front_02_R}
    componentAppliedLinearAccClamp=<2000 2000 2000>
    componentLinearAccScale=<0.2 0.2 0.2>
    componentLinearVelScale=<0.2 0.2 0.2>
    gravityOverride {z=-980}
    physicsBodyDefinitions [{
      boundBone {boneName=dyn_collar_front_02_R}
      boxExtents=<20 30 30>
      constraintSetup {angularLimitsMax=<0 5 5>, angularLimitsMin=<0 -5 -5>, angularTarget=<1 0 0>}
      localJointOffset=<3 0 0>
     }]
    sphericalLimits [
     {limitRadius=12, drivingBone {boneName=head}, sphereLocalOffset=<2 2 0>}
     {limitRadius=8, drivingBone {boneName=upperarm_r}, sphereLocalOffset=<-15 0 0>}
    ]
   }
   n15 = "Transform (Modify) Bone - Bone: dyn_Belt_MSTR"(ComponentPose=@n14, Alpha=@n1, Translation="0,-1,0", Rotation="0,0,0", Scale="1,1,1") {
    rotationMode=BMM_Additive
    rotationSpace=BCS_BoneSpace
    translationMode=BMM_Additive
    translationSpace=BCS_BoneSpace
    alphaBoolBlend {blendOption=Linear}
    boneToModify {boneName=dyn_Belt_MSTR}
    translation {y=-1}
   }
   n16 = Get bIsPlayingEmote()
   n17 = "Transform (Modify) Bone - Bone: dyn_jacket_front_L_01"!(ComponentPose=@n15, bAlphaBoolEnabled:bool=@n16, Translation="0,0.5,0.5", Rotation="0,0,0", Scale="1,1,1") -> (Pose) {
    alphaInputType=Bool
    translationMode=BMM_Additive
    translationSpace=BCS_BoneSpace
    alphaBoolBlend {blendOption=Linear}
    boneToModify {boneName=dyn_jacket_front_L_01}
    translation {y=0.5, z=0.5}
   }
   n18 = "Transform (Modify) Bone - Bone: dyn_jacket_front_R_01"!(ComponentPose=@n17, bAlphaBoolEnabled:bool=@n16, Translation="0,0.5,0.5", Rotation="0,0,0", Scale="1,1,1") -> (Pose) {
    alphaInputType=Bool
    translationMode=BMM_Additive
    translationSpace=BCS_BoneSpace
    alphaBoolBlend {blendOption=Linear}
    boneToModify {boneName=dyn_jacket_front_R_01}
    translation {y=0.5, z=0.5}
   }
   n19 = RigidBody!(ComponentPose=@n18, bAlphaBoolEnabled:bool, ExternalForce) -> (Pose) {
    alphaInputType=Bool
    bOverrideWorldGravity=true
    overridePhysicsAsset=&/F_MED_CandyCharm_Belt_Physics
    simulationSpace=BaseBoneSpace
    alphaBoolBlend {blendOption=Linear}
    baseBoneRef {boneName=pelvis}
    overrideWorldGravity {z=-980}
    simSpaceSettings {velocityScaleZ=0.2, worldAlpha=0.6}
   }
   n20 = RigidBody!(ComponentPose=@n19, Alpha:float=0.7, OverrideWorldGravity:Vector="0,0,-980", ExternalForce) -> (Pose) {
    bOverrideWorldGravity=true
    overridePhysicsAsset=&/F_MED_CandyCharm_Belt_FE_Physics
    simulationSpace=BaseBoneSpace
    alphaBoolBlend {blendOption=Linear}
    alphaScaleBiasClamp {bInterpResult=true}
    baseBoneRef {boneName=pelvis}
    overrideWorldGravity {z=-980}
    simSpaceSettings {velocityScaleZ=0.2}
   }
   n21 = RigidBody(ComponentPose=@n20, Alpha=0.7, OverrideWorldGravity="0,0,-980") {
    bOverrideWorldGravity=true
    overridePhysicsAsset=&/F_MED_CandyCharm_FE_Jacket_Physics
    simulationSpace=BaseBoneSpace
    alphaBoolBlend {blendOption=Linear}
    alphaScaleBiasClamp {bInterpResult=true}
    baseBoneRef {boneName=spine_03}
    overrideWorldGravity {z=-980}
    simSpaceSettings {velocityScaleZ=0.6, worldAlpha=1}
   }
   n22 = "Transform (Modify) Bone - Bone: dyn_base_Belt_Back"(ComponentPose=@n21, Alpha=1, Translation="1,0,2", Rotation="0,0,0", Scale="1,1,1") {
    rotationMode=BMM_Additive
    translationMode=BMM_Additive
    alphaBoolBlend {blendOption=Linear}
    alphaScaleBiasClamp {bInterpResult=true}
    boneToModify {boneName=dyn_base_Belt_Back}
    translation {x=1, z=2}
   }
   n23 = "Transform (Modify) Bone - Bone: dyn_Belt_MSTR"(ComponentPose=@n22, Alpha=1, Translation="1,0,0", Rotation="0,0,0", Scale="1,1,1") {
    rotationMode=BMM_Additive
    translationMode=BMM_Additive
    alphaBoolBlend {blendOption=Linear}
    alphaScaleBiasClamp {bInterpResult=true}
    boneToModify {boneName=dyn_Belt_MSTR}
    translation {x=1}
   }
   Output Pose(Result=Component To Local(ComponentPose=@n23)) {layerGroup=None}
  ]}
 EventGraph {nodes [
   Event BlueprintUpdateAnimation()
   > n1 = Gravity Override State Function DEPRECATED(CharacterPartAnimInstance=Self-Reference(), GameTimeInSeconds=GetGameTimeInSeconds(), GravityOverrideParams=Get Gravity Override Params())
   > Set GravityOverride(GravityOverride=@n1)
   > Set SkyDive(SkyDive:double="To Float (Boolean)"(InBool=Get bIsSkyDiving())) -> (Output_Get:double)
   n2 = Get bIsMoving2D()
   n3 = Get bIsAccelerating2D()
   n4 = Get bIsSprinting()
   > Set Moving(Moving:double="To Float (Boolean)"(InBool=OR Boolean!(A=@n2, B=@n3, C:bool=@n4) -> (ReturnValue))) -> (Output_Get:double)
   n5 = Get bIsCrouchSprinting()
   n6 = Get bIsCrouching()
   > Set Crouch(Crouch:double="To Float (Boolean)"(InBool=OR Boolean(A=@n5, B=@n6))) -> (Output_Get:double)
   n7 = NOT Boolean(A=Get bIsPlayingEmote())
   n8 = Get bIsInFrontEnd()
   > Set FE_DYN(FE_DYN:double="To Float (Boolean)"(InBool=AND Boolean(A=@n7, B=@n8))) -> (Output_Get:double)
   n9 = Get bIsSkyDiving()
   > Set SkyDive_Back_ADJ(SkyDive_Back_ADJ:double="To Float (Boolean)"(InBool=@n9)) -> (Output_Get:double)
   AND Boolean!(A=Get bIsSkyDivingFromBus() -> (bIsSkyDivingFromBus:bool), B=@n9, C:bool=Get bIsParachuteOpened()) -> (ReturnValue)
  ]}
]
variables [
 Crouch {type=number}
 FE_DYN {type=number}
 Gravity Override Params {type=GravityOverrideParamsStruct}
 GravityOverride {type=Vector}
 Moving {type=number}
 SkyDive {type=number}
 SkyDive_Back_ADJ {type=number}
]
