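# Front-end physics asset for the CandyCharm female medium face accessory, simulating hair and earrings.
# Capsule bodies cover branching hair chains, three earrings, collar pieces, head, neck and upper spine.
# Hair strands mostly branch from dyn_hair_02, which hangs off a root locked to the head; joints there and on earrings have swing limits and zero twist.
# Collar, spine and earring-base bodies have no constraints to the body, and head-to-neck rotation is free.
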PhysicsAsset /BRCosmetics/Characters/Player/Female/Medium/Bodies/F_MED_CandyCharm/Meshes/Parts/F_MED_CandyCharm_FaceAcc_FrontEndPhysics
bodies {
 base_earring_01_l [{shapeType=Capsule, length=0.505, radius=0.505, shapeName=base_earring_01_l_capsule, center=<0 0 0.1762>, rotation=<rot 0 -90 0>}]
 base_earring_01_r [{shapeType=Capsule, length=0.505, radius=0.505, shapeName=base_earring_01_r_capsule, center=<0 0 0.1762>, rotation=<rot 0 -90 0>}]
 base_earring_a_01_r [{shapeType=Capsule, length=0.505, radius=0.505, shapeName=base_earring_a_01_r_capsule, center=<0 0 0.1762>, rotation=<rot 0 -90 0>}]
 base_hair [{shapeType=Capsule, length=0.505, radius=0.505, shapeName=base_hair_capsule, rotation=<rot 0 -90 0>}]
 dyn_collar_Back_01 [{shapeType=Capsule, length=12, radius=3.5, shapeName=dyn_collar_Back_01_capsule, center=<3 0 -0.8>, rotation=<rot 0 0 90>}]
 dyn_collar_front_01_R [{shapeType=Capsule, length=12, radius=3.5, shapeName=dyn_collar_front_01_R_capsule, center=<4.9898 -3.1516 1.8629>, rotation=<rot 0 -10 90>}]
 dyn_collar_front_02_L [{shapeType=Capsule, length=12, radius=3.5, shapeName=dyn_collar_front_02_L_capsule, center=<-0.3503 3.7504 1.4893>, rotation=<rot 0 30 90>}]
 dyn_earring_01_l [{shapeType=Capsule, length=5, radius=1.2418, shapeName=dyn_earring_01_l_capsule, center=<2.6425 0 0>, rotation=<rot 0 90 -90>}]
 dyn_earring_01_r [{
   shapeType=Capsule
   length=4.5092
   radius=0.7653
   shapeName=dyn_earring_01_r_capsule
   center=<2.0844 0.3728 -0.1338>
   rotation=<rot 0 -91.6669 -86.3067>
  }]
 dyn_earring_a_01_r [{
   shapeType=Capsule
   length=6.7938
   radius=0.7603
   shapeName=dyn_earring_a_01_r_capsule
   center=<3.5965 -0.3274 -0.0323>
   rotation=<rot 0 -92.3124 -89.6779>
  }]
 dyn_hair_01 [{
   shapeType=Capsule
   length=4.2515
   radius=4.6801
   shapeName=dyn_hair_01_capsule
   center=<5.2777 -1.3668 -2.3868>
   rotation=<rot 0 -91.1331 -57.8296>
  }]
 dyn_hair_02 [{
   shapeType=Capsule
   length=11.2659
   radius=6.1866
   shapeName=dyn_hair_02_capsule
   center=<4.6258 -0.087 -2.0189>
   rotation=<rot 0 5.8915 -77.7313>
  }]
 dyn_hair_a_01_l [{
   shapeType=Capsule
   length=3.4266
   radius=1.3024
   shapeName=dyn_hair_a_01_l_capsule
   center=<2.4386 -0.2309 0.0023>
   rotation=<rot -12.3574 88.0449 -83.4409>
  }]
 dyn_hair_a_01_m [{
   shapeType=Capsule
   length=2.1517
   radius=1.0645
   shapeName=dyn_hair_a_01_m_capsule
   center=<2.6347 0.2348 0.0411>
   rotation=<rot -10.4803 89.8449 -86.8485>
  }]
 dyn_hair_a_02_l [{
   shapeType=Capsule
   length=4.1519
   radius=1.5401
   shapeName=dyn_hair_a_02_l_capsule
   center=<3.9083 0.0063 0.513>
   rotation=<rot 0 97.8489 -79.6895>
  }]
 dyn_hair_a_02_m [{
   shapeType=Capsule
   length=2.6433
   radius=1.282
   shapeName=dyn_hair_a_02_m_capsule
   center=<1.6309 0.1429 0.0943>
   rotation=<rot 0 88.7438 -89.5542>
  }]
 dyn_hair_a_03_m [{
   shapeType=Capsule
   length=2.4153
   radius=1.4252
   shapeName=dyn_hair_a_03_m_capsule
   center=<1.7744 0.2686 0.326>
   rotation=<rot -0.7205 81.8949 -88.5214>
  }]
 dyn_hair_b_01_l [{
   shapeType=Capsule
   length=2.5408
   radius=1.3208
   shapeName=dyn_hair_b_01_l_capsule
   center=<3.1387 0.1322 0.0061>
   rotation=<rot 0.353 -86.8351 -86.3347>
  }]
 dyn_hair_b_01_m [{
   shapeType=Capsule
   length=5.9094
   radius=1.5043
   shapeName=dyn_hair_b_01_m_capsule
   center=<3.8004 0.0855 -0.3382>
   rotation=<rot 1.4893 78.5706 -85.5895>
  }]
 dyn_hair_b_01_r [{
   shapeType=Capsule
   length=3.0504
   radius=1.2772
   shapeName=dyn_hair_b_01_r_capsule
   center=<1.637 0.0286 -0.2437>
   rotation=<rot 6.3472 95.8591 -85.6753>
  }]
 dyn_hair_b_01a_r [{
   shapeType=Capsule
   length=3.218
   radius=1.4666
   shapeName=dyn_hair_b_01a_r_capsule
   center=<4.4402 -0.2819 -0.0536>
   rotation=<rot 0 -84.2946 -89.0512>
  }]
 dyn_hair_b_02_l [{
   shapeType=Capsule
   length=6.2058
   radius=1.2854
   shapeName=dyn_hair_b_02_l_capsule
   center=<1.3252 0.1384 -0.0817>
   rotation=<rot 1.9902 -87.1779 -86.4558>
  }]
 dyn_hair_b_02_m [{
   shapeType=Capsule
   length=7.0311
   radius=1.5758
   shapeName=dyn_hair_b_02_m_capsule
   center=<4.8519 -0.171 0.2062>
   rotation=<rot -0.0101 91.1217 -83.1908>
  }]
 dyn_hair_b_02_r [{
   shapeType=Capsule
   length=3.6497
   radius=1.5332
   shapeName=dyn_hair_b_02_r_capsule
   center=<2.0194 0.1493 0.0254>
   rotation=<rot 18.6219 100.4399 -82.7706>
  }]
 dyn_hair_b_02a_r [{
   shapeType=Capsule
   length=2.7852
   radius=1.507
   shapeName=dyn_hair_b_02a_r_capsule
   center=<3.1891 -0.2194 -0.1058>
   rotation=<rot 0 97.7571 -80.9582>
  }]
 dyn_hair_c_01_l [{
   shapeType=Capsule
   length=2.1434
   radius=1.1598
   shapeName=dyn_hair_c_01_l_capsule
   center=<4.0893 -0.2397 -0.2202>
   rotation=<rot -4.3523 72.8802 -90.1511>
  }]
 dyn_hair_c_01_m [{
   shapeType=Capsule
   length=2.6506
   radius=1.4996
   shapeName=dyn_hair_c_01_m_capsule
   center=<3.378 -0.2046 -0.3311>
   rotation=<rot 9.6232 90.7583 -94.4467>
  }]
 dyn_hair_c_01_r [{
   shapeType=Capsule
   length=3.098
   radius=1.4183
   shapeName=dyn_hair_c_01_r_capsule
   center=<1.293 0.0662 -0.087>
   rotation=<rot 0 85 -85.7391>
  }]
 dyn_hair_c_01a_m [{
   shapeType=Capsule
   length=3.8651
   radius=1.8678
   shapeName=dyn_hair_c_01a_m_capsule
   center=<3.198 0.0926 -0.3691>
   rotation=<rot -3.7419 -88.9575 -86.9367>
  }]
 dyn_hair_c_01a_r [{
   shapeType=Capsule
   length=2.9448
   radius=1.6562
   shapeName=dyn_hair_c_01a_r_capsule
   center=<2.5273 -0.0745 -0.1507>
   rotation=<rot 0 -88.8126 -86.0749>
  }]
 dyn_hair_c_01aa_r [{
   shapeType=Capsule
   length=4.3073
   radius=1.7543
   shapeName=dyn_hair_c_01aa_r_capsule
   center=<3.6135 -0.5427 -0.0312>
   rotation=<rot -0.3462 -106.6577 -86.0373>
  }]
 dyn_hair_c_02_l [{
   shapeType=Capsule
   length=2.3703
   radius=0.9615
   shapeName=dyn_hair_c_02_l_capsule
   center=<1.7484 -0.5189 -0.1452>
   rotation=<rot -1.491 89.527 -90.8327>
  }]
 dyn_hair_c_02_m [{
   shapeType=Capsule
   length=4.1539
   radius=1.5363
   shapeName=dyn_hair_c_02_m_capsule
   center=<2.4906 -0.5182 -0.6439>
   rotation=<rot 15.8602 85.742 -88.2153>
  }]
 dyn_hair_c_02_r [{
   shapeType=Capsule
   length=3.8529
   radius=1.6782
   shapeName=dyn_hair_c_02_r_capsule
   center=<2.4465 -0.2706 0.6143>
   rotation=<rot 0 84.1932 -82.1118>
  }]
 dyn_hair_c_03_l [{
   shapeType=Capsule
   length=3.7809
   radius=1.3393
   shapeName=dyn_hair_c_03_l_capsule
   center=<3.0902 -0.3863 0.088>
   rotation=<rot 1.6067 91.5672 -91.2781>
  }]
 dyn_hair_d_01_l [{
   shapeType=Capsule
   length=2.2961
   radius=1.2426
   shapeName=dyn_hair_d_01_l_capsule
   center=<1.7491 0.5872 -0.0735>
   rotation=<rot -11.6877 89.8626 -91.8417>
  }]
 dyn_hair_d_01_r [{
   shapeType=Capsule
   length=2.9657
   radius=1.6603
   shapeName=dyn_hair_d_01_r_capsule
   center=<3.3952 -0.711 0.1046>
   rotation=<rot 90 179.9988 179.9986>
  }]
 dyn_hair_d_02_l [{
   shapeType=Capsule
   length=2.4898
   radius=1.2147
   shapeName=dyn_hair_d_02_l_capsule
   center=<1.7822 -0.196 -0.0106>
   rotation=<rot 0 87.2894 -85.5285>
  }]
 dyn_hair_d_02_r [{
   shapeType=Capsule
   length=2.8732
   radius=1.4286
   shapeName=dyn_hair_d_02_r_capsule
   center=<5.2808 -0.1299 -0.1114>
   rotation=<rot 0 -84.9717 91.9828>
  }]
 dyn_hair_d_03_l [{
   shapeType=Capsule
   length=2.9129
   radius=1.281
   shapeName=dyn_hair_d_03_l_capsule
   center=<2.2786 -0.4793 0.3136>
   rotation=<rot -28.9698 94.6566 -89.3911>
  }]
 dyn_hair_d_03_r [{
   shapeType=Capsule
   length=2.5234
   radius=1.5219
   shapeName=dyn_hair_d_03_r_capsule
   center=<4.5498 -0.0588 -0.1261>
   rotation=<rot 0 -91.5677 -87.7843>
  }]
 dyn_hair_d_04_r [{
   shapeType=Capsule
   length=2.3804
   radius=1.5166
   shapeName=dyn_hair_d_04_r_capsule
   center=<1.9943 0.4425 -0.1784>
   rotation=<rot -0.8527 -92.1219 -87.6562>
  }]
 dyn_hair_e_01_l [{
   shapeType=Capsule
   length=2.1949
   radius=1.5247
   shapeName=dyn_hair_e_01_l_capsule
   center=<3.1768 -0.191 -0.9504>
   rotation=<rot 0 -91.3944 -62.1593>
  }]
 dyn_hair_e_01_r [{
   shapeType=Capsule
   length=3.5115
   radius=1.5545
   shapeName=dyn_hair_e_01_r_capsule
   center=<4.2494 -0.3302 -0.1823>
   rotation=<rot 0 96.9253 -86.3809>
  }]
 dyn_hair_e_02_r [{
   shapeType=Capsule
   length=3.0017
   radius=1.6335
   shapeName=dyn_hair_e_02_r_capsule
   center=<4.8617 0.087 -0.1243>
   rotation=<rot 0 -86.1557 -89.1121>
  }]
 dyn_hair_e_03_r [{
   shapeType=Capsule
   length=3.6774
   radius=1.6001
   shapeName=dyn_hair_e_03_r_capsule
   center=<4.175 -0.2176 -0.0169>
   rotation=<rot -0.0838 -93.343 -89.5253>
  }]
 dyn_hair_e_04_r [{
   shapeType=Capsule
   length=1.9901
   radius=1.4131
   shapeName=dyn_hair_e_04_r_capsule
   center=<1.761 0.2778 -0.0815>
   rotation=<rot 0 -85.2822 -79.6941>
  }]
 head [{shapeType=Capsule, length=0, radius=9.9671, shapeName=head_capsule, center=<2.462 1.0229 0.4809>, rotation=<rot 7.8243 179.9999 -180>}]
 neck_02 [{shapeType=Capsule, length=6.253, radius=4.9714, shapeName=neck_02_capsule, center=<5.0249 0 0.4776>, rotation=<rot 0 -90 88.9316>}]
 spine_05 [{shapeType=Capsule, length=22.012, radius=12.6665, shapeName=spine_05_capsule, center=<0.6167 -0.3834 0>, rotation=<rot 0 -90 0>}]
}
constraints [
 {
  bone1Name=head
  bone2Name=neck_02
  swing1LimitDegrees=45
  swing1Motion=Free
  swing2LimitDegrees=45
  swing2Motion=Free
  twistLimitDegrees=45
  twistMotion=Free
 }
 {bone1Name=dyn_hair_e_01_l, bone2Name=dyn_hair_01, swing1LimitDegrees=20, swing2LimitDegrees=20, twistLimitDegrees=0}
 {bone1Name=dyn_hair_d_03_l, bone2Name=dyn_hair_d_02_l, swing1LimitDegrees=30, swing2LimitDegrees=20, twistLimitDegrees=0}
 {bone1Name=dyn_hair_d_02_l, bone2Name=dyn_hair_d_01_l, swing1LimitDegrees=30, swing2LimitDegrees=20, twistLimitDegrees=0}
 {bone1Name=dyn_hair_d_01_l, bone2Name=dyn_hair_02, swing1LimitDegrees=20, swing2LimitDegrees=20, twistLimitDegrees=0}
 {bone1Name=dyn_hair_c_02_r, bone2Name=dyn_hair_c_01_r, swing1LimitDegrees=30, swing2LimitDegrees=20, twistLimitDegrees=0}
 {bone1Name=dyn_hair_c_01_r, bone2Name=dyn_hair_c_01a_r, swing1LimitDegrees=40, swing2LimitDegrees=20, twistLimitDegrees=0}
 {bone1Name=dyn_hair_c_02_m, bone2Name=dyn_hair_c_01_m, swing1LimitDegrees=30, swing2LimitDegrees=20, twistLimitDegrees=0}
 {bone1Name=dyn_hair_c_01_m, bone2Name=dyn_hair_c_01a_m, swing1LimitDegrees=40, swing2LimitDegrees=20, twistLimitDegrees=0}
 {bone1Name=dyn_hair_c_03_l, bone2Name=dyn_hair_c_02_l, swing1LimitDegrees=30, swing2LimitDegrees=20, twistLimitDegrees=0}
 {bone1Name=dyn_hair_c_02_l, bone2Name=dyn_hair_c_01_l, swing1LimitDegrees=30, swing2LimitDegrees=20, twistLimitDegrees=0}
 {bone1Name=dyn_hair_c_01_l, bone2Name=dyn_hair_02, swing1LimitDegrees=40, swing2LimitDegrees=20, twistLimitDegrees=0}
 {bone1Name=dyn_hair_b_02_r, bone2Name=dyn_hair_b_01_r, swing1LimitDegrees=30, swing2LimitDegrees=20, twistLimitDegrees=0}
 {bone1Name=dyn_hair_b_01_r, bone2Name=dyn_hair_b_02a_r, swing1LimitDegrees=30, swing2LimitDegrees=10, twistLimitDegrees=0}
 {bone1Name=dyn_hair_b_02_m, bone2Name=dyn_hair_b_01_m, swing1LimitDegrees=30, swing2LimitDegrees=20, twistLimitDegrees=0}
 {bone1Name=dyn_hair_b_01_m, bone2Name=dyn_hair_02, swing1LimitDegrees=40, swing2LimitDegrees=20, twistLimitDegrees=0}
 {bone1Name=dyn_hair_b_02_l, bone2Name=dyn_hair_b_01_l, swing1LimitDegrees=30, swing2LimitDegrees=20, twistLimitDegrees=0}
 {bone1Name=dyn_hair_b_01_l, bone2Name=dyn_hair_02, swing1LimitDegrees=40, swing2LimitDegrees=20, twistLimitDegrees=0}
 {bone1Name=dyn_hair_a_03_m, bone2Name=dyn_hair_a_02_m, swing1LimitDegrees=30, swing2LimitDegrees=20, twistLimitDegrees=0}
 {bone1Name=dyn_hair_a_02_m, bone2Name=dyn_hair_a_01_m, swing1LimitDegrees=30, swing2LimitDegrees=20, twistLimitDegrees=0}
 {bone1Name=dyn_hair_a_01_m, bone2Name=dyn_hair_02, swing1LimitDegrees=20, swing2LimitDegrees=20, twistLimitDegrees=0}
 {bone1Name=dyn_hair_a_02_l, bone2Name=dyn_hair_a_01_l, swing1LimitDegrees=20, swing2LimitDegrees=20, twistLimitDegrees=0}
 {bone1Name=dyn_hair_a_01_l, bone2Name=dyn_hair_02, swing1LimitDegrees=40, swing2LimitDegrees=20, twistLimitDegrees=0}
 {bone1Name=dyn_hair_02, bone2Name=dyn_hair_01, swing1LimitDegrees=1, swing2LimitDegrees=50, twistLimitDegrees=0}
 {
  bone1Name=dyn_hair_01
  bone2Name=base_hair
  swing1LimitDegrees=45
  swing1Motion=Locked
  swing2LimitDegrees=45
  swing2Motion=Locked
  twistLimitDegrees=45
  twistMotion=Locked
 }
 {
  bone1Name=base_hair
  bone2Name=head
  swing1LimitDegrees=45
  swing1Motion=Locked
  swing2LimitDegrees=45
  swing2Motion=Locked
  twistLimitDegrees=45
  twistMotion=Locked
 }
 {bone1Name=dyn_hair_b_01a_r, bone2Name=dyn_hair_02, swing1LimitDegrees=15, swing2LimitDegrees=30, twistLimitDegrees=0}
 {bone1Name=dyn_hair_b_02a_r, bone2Name=dyn_hair_b_01a_r, swing1LimitDegrees=15, swing2LimitDegrees=30, twistLimitDegrees=0}
 {bone1Name=dyn_hair_c_01a_r, bone2Name=dyn_hair_c_01aa_r, swing1LimitDegrees=40, swing2LimitDegrees=15, twistLimitDegrees=0}
 {bone1Name=dyn_hair_e_01_r, bone2Name=dyn_hair_02, swing1LimitDegrees=30, swing2LimitDegrees=10, twistLimitDegrees=0}
 {bone1Name=dyn_hair_e_02_r, bone2Name=dyn_hair_e_01_r, swing1LimitDegrees=20, swing2LimitDegrees=10, twistLimitDegrees=0}
 {bone1Name=dyn_hair_e_03_r, bone2Name=dyn_hair_e_02_r, swing1LimitDegrees=20, swing2LimitDegrees=10, twistLimitDegrees=0}
 {bone1Name=dyn_hair_e_04_r, bone2Name=dyn_hair_e_03_r, swing1LimitDegrees=10, swing2LimitDegrees=5, twistLimitDegrees=0}
 {bone1Name=dyn_hair_c_01a_m, bone2Name=dyn_hair_02, swing1LimitDegrees=30, swing2LimitDegrees=10, twistLimitDegrees=0}
 {bone1Name=dyn_hair_c_01aa_r, bone2Name=dyn_hair_02, swing1LimitDegrees=10, swing2LimitDegrees=15, twistLimitDegrees=0}
 {bone1Name=dyn_hair_d_01_r, bone2Name=dyn_hair_02, swing1LimitDegrees=30, swing2LimitDegrees=10, twistLimitDegrees=0}
 {bone1Name=dyn_hair_d_02_r, bone2Name=dyn_hair_d_01_r, swing1LimitDegrees=20, swing2LimitDegrees=10, twistLimitDegrees=0}
 {bone1Name=dyn_hair_d_03_r, bone2Name=dyn_hair_d_02_r, swing1LimitDegrees=15, swing2LimitDegrees=10, twistLimitDegrees=0}
 {bone1Name=dyn_hair_d_04_r, bone2Name=dyn_hair_d_03_r, swing1LimitDegrees=10, swing2LimitDegrees=5, twistLimitDegrees=0}
 {bone1Name=dyn_earring_01_l, bone2Name=base_earring_01_l, swing1LimitDegrees=30, swing2LimitDegrees=10, twistLimitDegrees=0}
 {bone1Name=dyn_earring_01_r, bone2Name=base_earring_01_r, swing1LimitDegrees=30, swing2LimitDegrees=30, twistLimitDegrees=0}
 {bone1Name=dyn_earring_a_01_r, bone2Name=base_earring_a_01_r, swing1LimitDegrees=30, swing2LimitDegrees=30, twistLimitDegrees=0}
]
