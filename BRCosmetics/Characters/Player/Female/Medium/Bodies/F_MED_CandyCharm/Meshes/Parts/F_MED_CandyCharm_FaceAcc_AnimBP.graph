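# Animation Blueprint for the CandyCharm female character's FaceAcc part, adding secondary hair motion to the pose copied from its attached parent mesh.
# A skydiving bool blends between regular and skydive RigidBody setups.
# Additive bone offsets, a Control Rig, front-hair AnimDynamics, a crouch-gated offset and a front-end RigidBody follow in series, then final bone tweaks.
# The event graph sets GravityOverride through a deprecated helper, plus a Skydive float.
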AnimBlueprint /BRCosmetics/Characters/Player/Female/Medium/Bodies/F_MED_CandyCharm/Meshes/Parts/F_MED_CandyCharm_FaceAcc_AnimBP
target_skeleton=&/F_MED_CandyCharm_FaceAcc_Skeleton
graphs [
 AnimGraph {nodes [
   Anim Dynamics - Bone: dyn_hair_e_01_l!(ComponentPose, Alpha=1, GravityOverride:Vector="0,0,-980", ExternalForce="0,0,0") -> (Pose) {
    angularDampingOverride=1
    angularSpringConstant=20000
    bAngularSpring=true
    bOverrideAngularDamping=true
    bUseGravityOverride=true
    gravityScale=0.35
    simulationSpace=RootRelative
    alphaBoolBlend {blendOption=Linear}
    boundBone {boneName=dyn_hair_e_01_l}
    componentLinearAccScale=<1 1 1>
    componentLinearVelScale=<1 1 1>
    gravityOverride {z=-980}
    physicsBodyDefinitions [{
      boundBone {boneName=dyn_hair_e_01_l}
      boxExtents=<20 30 30>
      constraintSetup {angularLimitsMax=<15 90 15>, angularLimitsMin=<-15 -90 -15>, angularTarget=<1 0 0>}
      localJointOffset=<4 0 0>
     }]
   }
   "Anim Dynamics - Chain: dyn_hair_e_l -> dyn_hair_e_l"!(ComponentPose, Alpha=1, GravityOverride:Vector="0,0,-980", ExternalForce="0,0,0") -> (Pose) {
    angularDampingOverride=0.9999
    angularSpringConstant=20000
    bAngularSpring=true
    bChain=true
    bOverrideAngularDamping=true
    bUseGravityOverride=true
    gravityScale=0.35
    simulationSpace=RootRelative
    alphaBoolBlend {blendOption=Linear}
    boundBone {boneName=dyn_hair_e_l}
    chainEnd {boneName=dyn_hair_e_l}
    componentLinearAccScale=<1 1 1>
    componentLinearVelScale=<1 1 1>
    gravityOverride {z=-980}
    physicsBodyDefinitions [{
      boundBone {boneName=dyn_hair_e_l}
      boxExtents=<20 30 30>
      constraintSetup {angularLimitsMax=<15 15 15>, angularLimitsMin=<-15 -15 -15>, angularTarget=<1 0 0>}
      localJointOffset=<4 0 0>
     }]
   }
   n1 = Anim Dynamics - Bone: dyn_hair_02(ComponentPose=Local To Component(LocalPose=Use cached pose 'InputPose'()), Alpha=1) {
    angularDampingOverride=1
    angularSpringConstant=20000
    bAngularSpring=true
    bOverrideAngularDamping=true
    bUseGravityOverride=true
    gravityScale=0.35
    simulationSpace=RootRelative
    alphaBoolBlend {blendOption=Linear}
    boundBone {boneName=dyn_hair_02}
    componentAppliedLinearAccClamp=<1500 1500 1500>
    componentLinearAccScale=<0.5 0.5 0.5>
    componentLinearVelScale=<0.5 0.5 0.5>
    gravityOverride {z=-980}
    physicsBodyDefinitions [{
      boundBone {boneName=dyn_hair_02}
      boxExtents=<20 30 30>
      constraintSetup {angularLimitsMax=<10 80 5>, angularLimitsMin=<-10 0 -5>, angularTarget=<1 0 0>}
      localJointOffset=<4 0 0>
     }]
   }
   n2 = RigidBody!(ComponentPose=@n1, Alpha, OverrideWorldGravity:Vector="0,0,0", ExternalForce) -> (Pose) {
    bOverrideWorldGravity=true
    overridePhysicsAsset=&/F_MED_CandyCharm_FaceAcc_Physics
    simulationSpace=BaseBoneSpace
    alphaBoolBlend {blendOption=Linear}
    baseBoneRef {boneName=head}
    componentAppliedLinearAccClamp=<2000 2000 2000>
    overrideWorldGravity {z=-980}
    simSpaceSettings {maxAngularAcceleration=250, maxAngularVelocity=20, maxLinearAcceleration=4500, maxLinearVelocity=900, velocityScaleZ=0.4, worldAlpha=0.65}
   }
   n3 = Component To Local(ComponentPose=@n2)
   n4 = RigidBody(ComponentPose=Local To Component(LocalPose=Use cached pose 'InputPose'()), OverrideWorldGravity="0,0,100") {
    bOverrideWorldGravity=true
    overridePhysicsAsset=&/F_MED_CandyCharm_FaceAcc_SKydive_Physics.F_MED_CandyCharm_FaceAcc_skydive_Physics
    simulationSpace=BaseBoneSpace
    alphaBoolBlend {blendOption=Linear}
    baseBoneRef {boneName=neck_02}
    componentAppliedLinearAccClamp=<2000 2000 2000>
    overrideWorldGravity {z=-980}
    simSpaceSettings {dampingAlpha=0.2, velocityScaleZ=0.45, worldAlpha=0.5}
   }
   n5 = Component To Local(ComponentPose=@n4)
   n6 = Get bIsSkyDiving()
   n7 = "Transform (Modify) Bone - Bone: dyn_hair_02"(
    ComponentPose=Local To Component(LocalPose=Blend Poses by bool(BlendPose_0=@n5, BlendPose_1=@n3, bActiveValue=@n6) {blendType=Linear})
    Alpha=1
    Translation="0,0,0"
    Rotation="0,0,10"
    Scale="1,1,1"
   ) {
    rotationMode=BMM_Additive
    rotationSpace=BCS_BoneSpace
    scaleSpace=BCS_BoneSpace
    translationMode=BMM_Additive
    translationSpace=BCS_BoneSpace
    alphaBoolBlend {blendOption=Linear}
    boneToModify {boneName=dyn_hair_02}
    rotation {roll=10}
   }
   n8 = "Transform (Modify) Bone - Bone: dyn_hair_01"(ComponentPose=@n7, Alpha=1, Translation="0,0,0", Rotation="0,0,0", Scale="1,1,1") {rotationMode=BMM_Additive, translationMode=BMM_Additive, alphaBoolBlend {blendOption=Linear}, boneToModify {boneName=dyn_hair_01}}
   n9 = "Transform (Modify) Bone - Bone: dyn_hair_b_01a_r"(ComponentPose=@n8, Alpha=1, Translation="0,0,0", Rotation="0,0,5", Scale="1,1,1") {
    rotationMode=BMM_Additive
    translationMode=BMM_Additive
    alphaBoolBlend {blendOption=Linear}
    boneToModify {boneName=dyn_hair_b_01a_r}
    rotation {roll=5}
   }
   n10 = "Transform (Modify) Bone - Bone: dyn_hair_b_01_m"(ComponentPose=@n9, Alpha=1, Translation="0,0,0", Rotation="-10,0,20", Scale="1,1,1") {
    rotationMode=BMM_Additive
    translationMode=BMM_Additive
    alphaBoolBlend {blendOption=Linear}
    boneToModify {boneName=dyn_hair_b_01_m}
    rotation {pitch=-10, roll=20}
   }
   n11 = "Transform (Modify) Bone - Bone: dyn_hair_b_01_l"(ComponentPose=@n10, Alpha=1, Translation="0,0,0", Rotation="0,0,10", Scale="1,1,1") {
    rotationMode=BMM_Additive
    translationMode=BMM_Additive
    alphaBoolBlend {blendOption=Linear}
    boneToModify {boneName=dyn_hair_b_01_l}
    rotation {roll=10}
   }
   n12 = "Transform (Modify) Bone - Bone: dyn_hair_d_01_l"(ComponentPose=@n11, Alpha=1, Translation="0,0,0", Rotation="0,0,10", Scale="1,1,1") {
    rotationMode=BMM_Additive
    translationMode=BMM_Additive
    alphaBoolBlend {blendOption=Linear}
    boneToModify {boneName=dyn_hair_d_01_l}
    rotation {roll=10}
   }
   n13 = "Transform (Modify) Bone - Bone: dyn_hair_d_01_r"(ComponentPose=@n12, Alpha=1, Translation="0,0,0", Rotation="10,0,10", Scale="1,1,1") {
    rotationMode=BMM_Additive
    translationMode=BMM_Additive
    alphaBoolBlend {blendOption=Linear}
    boneToModify {boneName=dyn_hair_d_01_r}
    rotation {pitch=10, roll=10}
   }
   n14 = "Transform (Modify) Bone - Bone: dyn_hair_a_01_l"(ComponentPose=@n13, Alpha=1, Translation="0,0,0", Rotation="0,0,10", Scale="1,1,1") {
    rotationMode=BMM_Additive
    translationMode=BMM_Additive
    alphaBoolBlend {blendOption=Linear}
    boneToModify {boneName=dyn_hair_a_01_l}
    rotation {roll=10}
   }
   n15 = "Transform (Modify) Bone - Bone: dyn_hair_c_01_l"(ComponentPose=@n14, Alpha=1, Translation="0,0,0", Rotation="-10,0,10", Scale="1,1,1") {
    rotationMode=BMM_Additive
    translationMode=BMM_Additive
    alphaBoolBlend {blendOption=Linear}
    boneToModify {boneName=dyn_hair_c_01_l}
    rotation {pitch=-10, roll=10}
   }
   n16 = "Transform (Modify) Bone - Bone: dyn_hair_a_01_m"(ComponentPose=@n15, Alpha=1, Translation="0,0,0", Rotation="0,0,10", Scale="1,1,1") {
    rotationMode=BMM_Additive
    translationMode=BMM_Additive
    alphaBoolBlend {blendOption=Linear}
    boneToModify {boneName=dyn_hair_a_01_m}
    rotation {roll=10}
   }
   n17 = "Transform (Modify) Bone - Bone: dyn_hair_b_01a_r"(ComponentPose=@n16, Alpha=1, Translation="0,0,0", Rotation="0,0,10", Scale="1,1,1") {
    rotationMode=BMM_Additive
    translationMode=BMM_Additive
    alphaBoolBlend {blendOption=Linear}
    boneToModify {boneName=dyn_hair_b_01a_r}
    rotation {roll=10}
   }
   n18 = "Transform (Modify) Bone - Bone: dyn_hair_a_01a_r"(ComponentPose=@n17, Alpha=1, Translation="0,0,0", Rotation="0,0,10", Scale="1,1,1") {
    rotationMode=BMM_Additive
    translationMode=BMM_Additive
    alphaBoolBlend {blendOption=Linear}
    boneToModify {boneName=dyn_hair_a_01a_r}
    rotation {roll=10}
   }
   n19 = "Transform (Modify) Bone - Bone: dyn_hair_e_01_r"(ComponentPose=@n18, Alpha=1, Translation="0,0,0", Rotation="20,0,20", Scale="1,1,1") {
    rotationMode=BMM_Additive
    translationMode=BMM_Additive
    alphaBoolBlend {blendOption=Linear}
    boneToModify {boneName=dyn_hair_e_01_r}
    rotation {pitch=20, roll=20}
   }
   n20 = "Transform (Modify) Bone - Bone: dyn_hair_c_01aa_r"(ComponentPose=@n19, Alpha=1, Translation="0,0,0", Rotation="0,0,10", Scale="1,1,1") {
    rotationMode=BMM_Additive
    translationMode=BMM_Additive
    alphaBoolBlend {blendOption=Linear}
    boneToModify {boneName=dyn_hair_c_01aa_r}
    rotation {roll=10}
   }
   n21 = "Transform (Modify) Bone - Bone: dyn_hair_c_01a_m"(ComponentPose=@n20, Alpha=1, Translation="0,0,0", Rotation="20,0,10", Scale="1,1,1") {
    rotationMode=BMM_Additive
    translationMode=BMM_Additive
    alphaBoolBlend {blendOption=Linear}
    boneToModify {boneName=dyn_hair_c_01a_m}
    rotation {pitch=20, roll=10}
   }
   n22 = Control Rig(Source=Component To Local(ComponentPose=@n21)) {
    alphaBoolBlend {blendOption=Linear}
    controlRigAssetReference {blueprintRigClass=&/F_MED_CandyCharm_FaceAcc_CtrlRig.F_MED_CandyCharm_FaceAcc_CtrlRig_C}
   }
   n23 = Anim Dynamics - Bone: dyn_front_hair_d_01_l(ComponentPose=Local To Component(LocalPose=@n22), Alpha=1, ExternalForce="0,0,0") {
    angularDampingOverride=0.9999
    angularSpringConstant=500
    bAngularSpring=true
    bOverrideAngularDamping=true
    bUseGravityOverride=true
    simulationSpace=RootRelative
    alphaBoolBlend {blendOption=Linear}
    boundBone {boneName=dyn_front_hair_d_01_l}
    componentAppliedLinearAccClamp=<2000 2000 2000>
    componentLinearAccScale=<0.2 0.2 0.2>
    componentLinearVelScale=<0.2 0.2 0.2>
    gravityOverride {z=-680}
    physicsBodyDefinitions [{
      boundBone {boneName=dyn_front_hair_d_01_l}
      boxExtents=<10 5 5>
      constraintSetup {angularLimitsMax=<0 0 10>, angularLimitsMin=<-10 -10 -10>, angularTarget=<0 0 0>}
      localJointOffset=<2.5 0 0>
     }]
   }
   n24 = Anim Dynamics - Bone: dyn_front_hair_c_01_l(ComponentPose=@n23, Alpha=1, ExternalForce="0,0,0") {
    angularDampingOverride=0.9999
    angularSpringConstant=500
    bAngularSpring=true
    bOverrideAngularDamping=true
    bUseGravityOverride=true
    simulationSpace=RootRelative
    alphaBoolBlend {blendOption=Linear}
    boundBone {boneName=dyn_front_hair_c_01_l}
    componentAppliedLinearAccClamp=<2000 2000 2000>
    componentLinearAccScale=<0.2 0.2 0.2>
    componentLinearVelScale=<0.2 0.2 0.2>
    gravityOverride {z=-680}
    physicsBodyDefinitions [{
      boundBone {boneName=dyn_front_hair_c_01_l}
      boxExtents=<10 5 5>
      constraintSetup {angularLimitsMax=<0 0 10>, angularLimitsMin=<-10 -10 -10>, angularTarget=<0 0 0>}
      localJointOffset=<2.5 0 0>
     }]
   }
   n25 = Anim Dynamics - Bone: dyn_front_hair_b_01_l(ComponentPose=@n24, Alpha=1, ExternalForce="0,0,0") {
    angularDampingOverride=0.9999
    angularSpringConstant=600
    bAngularSpring=true
    bOverrideAngularDamping=true
    bUseGravityOverride=true
    simulationSpace=RootRelative
    alphaBoolBlend {blendOption=Linear}
    boundBone {boneName=dyn_front_hair_b_01_l}
    componentAppliedLinearAccClamp=<2000 2000 2000>
    componentLinearAccScale=<0.2 0.2 0.2>
    componentLinearVelScale=<0.2 0.2 0.2>
    gravityOverride {z=-980}
    physicsBodyDefinitions [{
      boundBone {boneName=dyn_front_hair_b_01_l}
      boxExtents=<10 5 5>
      constraintSetup {angularLimitsMax=<5 0 5>, angularLimitsMin=<-5 0 -5>, angularTarget=<1 0 0>}
      localJointOffset=<1 0 0>
     }]
   }
   n26 = Anim Dynamics - Bone: dyn_front_hair_b_02_l(ComponentPose=@n25, Alpha=1, ExternalForce="0,0,0") {
    angularDampingOverride=0.999
    angularSpringConstant=600
    bAngularSpring=true
    bOverrideAngularDamping=true
    bUseGravityOverride=true
    simulationSpace=RootRelative
    alphaBoolBlend {blendOption=Linear}
    boundBone {boneName=dyn_front_hair_b_02_l}
    componentAppliedLinearAccClamp=<2000 2000 2000>
    componentLinearAccScale=<0.2 0.2 0.2>
    componentLinearVelScale=<0.2 0.2 0.2>
    gravityOverride {z=-980}
    physicsBodyDefinitions [{
      boundBone {boneName=dyn_front_hair_b_02_l}
      boxExtents=<15 8 8>
      constraintSetup {angularLimitsMax=<5 5 10>, angularLimitsMin=<-5 -5 -10>, angularTarget=<1 0 0>}
      localJointOffset=<2 0 0>
     }]
   }
   n27 = Anim Dynamics - Bone: dyn_front_hair_a_01_l(ComponentPose=@n26, Alpha=1, ExternalForce="0,0,0") {
    angularDampingOverride=0.999
    angularSpringConstant=600
    bAngularSpring=true
    bOverrideAngularDamping=true
    bUseGravityOverride=true
    simulationSpace=RootRelative
    alphaBoolBlend {blendOption=Linear}
    boundBone {boneName=dyn_front_hair_a_01_l}
    componentAppliedLinearAccClamp=<2000 2000 2000>
    componentLinearAccScale=<0.2 0.2 0.2>
    componentLinearVelScale=<0.2 0.2 0.2>
    gravityOverride {z=-980}
    physicsBodyDefinitions [{
      boundBone {boneName=dyn_front_hair_a_01_l}
      boxExtents=<15 8 8>
      constraintSetup {angularLimitsMax=<5 5 10>, angularLimitsMin=<-5 -5 -10>, angularTarget=<1 0 0>}
      localJointOffset=<2 0 0>
     }]
   }
   n28 = Anim Dynamics - Bone: dyn_front_hair_a_01_m(ComponentPose=@n27, Alpha=1, ExternalForce="0,0,0") {
    angularDampingOverride=0.9999
    angularSpringConstant=600
    bAngularSpring=true
    bOverrideAngularDamping=true
    bUseGravityOverride=true
    simulationSpace=RootRelative
    alphaBoolBlend {blendOption=Linear}
    boundBone {boneName=dyn_front_hair_a_01_m}
    componentAppliedLinearAccClamp=<2000 2000 2000>
    componentLinearAccScale=<0.2 0.2 0.2>
    componentLinearVelScale=<0.2 0.2 0.2>
    gravityOverride {z=-980}
    physicsBodyDefinitions [{
      boundBone {boneName=dyn_front_hair_a_01_m}
      boxExtents=<10 5 5>
      constraintSetup {angularLimitsMax=<5 0 5>, angularLimitsMin=<-5 0 0>, angularTarget=<1 0 0>}
      localJointOffset=<1 0 0>
     }]
   }
   n29 = Anim Dynamics - Bone: dyn_front_hair_a_02_m(ComponentPose=@n28, Alpha=1, ExternalForce="0,0,0") {
    angularDampingOverride=0.999
    angularSpringConstant=600
    bAngularSpring=true
    bOverrideAngularDamping=true
    bUseGravityOverride=true
    bUseSphericalLimits=true
    simulationSpace=RootRelative
    alphaBoolBlend {blendOption=Linear}
    boundBone {boneName=dyn_front_hair_a_02_m}
    componentAppliedLinearAccClamp=<2000 2000 2000>
    componentLinearAccScale=<0.2 0.2 0.2>
    componentLinearVelScale=<0.2 0.2 0.2>
    gravityOverride {z=-980}
    physicsBodyDefinitions [{
      boundBone {boneName=dyn_front_hair_a_02_m}
      boxExtents=<15 8 8>
      constraintSetup {angularLimitsMax=<5 0 10>, angularLimitsMin=<-5 0 0>, angularTarget=<1 0 0>}
      localJointOffset=<2 0 0>
     }]
    sphericalLimits [{limitRadius=9, drivingBone {boneName=None}, sphereLocalOffset=<0 2.3 0>}]
   }
   n30 = Anim Dynamics - Bone: dyn_front_hair_b_01_m(ComponentPose=@n29, Alpha=1, ExternalForce="0,0,0") {
    angularDampingOverride=0.9999
    angularSpringConstant=600
    bAngularSpring=true
    bOverrideAngularDamping=true
    bUseGravityOverride=true
    simulationSpace=RootRelative
    alphaBoolBlend {blendOption=Linear}
    boundBone {boneName=dyn_front_hair_b_01_m}
    componentAppliedLinearAccClamp=<2000 2000 2000>
    componentLinearAccScale=<0.2 0.2 0.2>
    componentLinearVelScale=<0.2 0.2 0.2>
    gravityOverride {z=-980}
    physicsBodyDefinitions [{
      boundBone {boneName=dyn_front_hair_b_01_m}
      boxExtents=<10 5 5>
      constraintSetup {angularLimitsMax=<5 0 5>, angularLimitsMin=<-5 0 -5>, angularTarget=<1 0 0>}
      localJointOffset=<1 0 0>
     }]
   }
   n31 = Anim Dynamics - Bone: dyn_front_hair_b_02_m(ComponentPose=@n30, Alpha=1, ExternalForce="0,0,0") {
    angularDampingOverride=0.999
    angularSpringConstant=600
    bAngularSpring=true
    bOverrideAngularDamping=true
    bUseGravityOverride=true
    simulationSpace=RootRelative
    alphaBoolBlend {blendOption=Linear}
    boundBone {boneName=dyn_front_hair_b_02_m}
    componentAppliedLinearAccClamp=<2000 2000 2000>
    componentLinearAccScale=<0.2 0.2 0.2>
    componentLinearVelScale=<0.2 0.2 0.2>
    gravityOverride {z=-980}
    physicsBodyDefinitions [{
      boundBone {boneName=dyn_front_hair_b_02_m}
      boxExtents=<15 8 8>
      constraintSetup {angularLimitsMax=<5 5 10>, angularLimitsMin=<-5 -5 -10>, angularTarget=<1 0 0>}
      localJointOffset=<2 0 0>
     }]
   }
   n32 = Anim Dynamics - Bone: dyn_front_hair_a_02_r(ComponentPose=@n31, Alpha=1, ExternalForce="0,0,0") {
    angularDampingOverride=0.999
    angularSpringConstant=600
    bAngularSpring=true
    bOverrideAngularDamping=true
    bUseGravityOverride=true
    simulationSpace=RootRelative
    alphaBoolBlend {blendOption=Linear}
    boundBone {boneName=dyn_front_hair_a_02_r}
    componentAppliedLinearAccClamp=<2000 2000 2000>
    componentLinearAccScale=<0.2 0.2 0.2>
    componentLinearVelScale=<0.2 0.2 0.2>
    gravityOverride {z=-980}
    physicsBodyDefinitions [{
      boundBone {boneName=dyn_front_hair_a_02_r}
      boxExtents=<15 8 8>
      constraintSetup {angularLimitsMax=<5 5 10>, angularLimitsMin=<-5 -5 -10>, angularTarget=<1 0 0>}
      localJointOffset=<2 0 0>
     }]
   }
   n33 = "Transform (Modify) Bone - Bone: dyn_hair_02"!(ComponentPose=@n32, bAlphaBoolEnabled:bool=Get bIsCrouching(), Translation="0,0,0", Rotation="-2,0,0", Scale="1,1,1") -> (Pose) {
    alphaInputType=Bool
    rotationMode=BMM_Additive
    rotationSpace=BCS_BoneSpace
    scaleMode=BMM_Additive
    scaleSpace=BCS_BoneSpace
    translationMode=BMM_Additive
    translationSpace=BCS_BoneSpace
    alphaBoolBlend {blendOption=Linear}
    alphaScaleBiasClamp {bInterpResult=true, interpSpeedDecreasing=30, interpSpeedIncreasing=30}
    boneToModify {boneName=dyn_hair_02}
    rotation {pitch=-2}
   }
   n34 = RigidBody!(ComponentPose=@n33, bAlphaBoolEnabled:bool, OverrideWorldGravity="0,0,0", ExternalForce) -> (Pose) {
    alphaInputType=Bool
    bOverrideWorldGravity=true
    overridePhysicsAsset=&/F_MED_CandyCharm_FaceAcc_FrontEndPhysics
    simulationSpace=BaseBoneSpace
    alphaBoolBlend {blendOption=Linear}
    baseBoneRef {boneName=head}
    overrideWorldGravity {z=-980}
    simSpaceSettings {maxAngularAcceleration=250, maxAngularVelocity=20, maxLinearAcceleration=4500, maxLinearVelocity=900, velocityScaleZ=0.4, worldAlpha=0.65}
   }
   n35 = "Transform (Modify) Bone - Bone: dyn_hair_d_02_r"!(ComponentPose=@n34, bAlphaBoolEnabled:bool=True, Translation="0,0,0", Rotation="0,0,-10", Scale="1,1,1") -> (Pose) {
    alphaInputType=Bool
    rotationMode=BMM_Additive
    translationMode=BMM_Additive
    alphaBoolBlend {blendOption=Linear}
    boneToModify {boneName=dyn_hair_d_02_r}
    rotation {roll=-10}
   }
   n36 = "Transform (Modify) Bone - Bone: dyn_hair_d_01_r"!(ComponentPose=@n35, bAlphaBoolEnabled:bool=True, Translation="0,1,0", Rotation="2,0,5", Scale="1,1,1") -> (Pose) {
    alphaInputType=Bool
    rotationMode=BMM_Additive
    translationMode=BMM_Additive
    alphaBoolBlend {blendOption=Linear}
    boneToModify {boneName=dyn_hair_d_01_r}
    rotation {pitch=2, roll=5}
    translation {y=1}
   }
   n37 = "Transform (Modify) Bone - Bone: dyn_hair_e_01_r"!(ComponentPose=@n36, bAlphaBoolEnabled:bool=True, Translation="0,-1,-2", Rotation="0,0,-2", Scale="1,1,1") -> (Pose) {
    alphaInputType=Bool
    rotationMode=BMM_Additive
    translationMode=BMM_Additive
    alphaBoolBlend {blendOption=Linear}
    boneToModify {boneName=dyn_hair_e_01_r}
    rotation {roll=-2}
    translation {y=-1, z=-2}
   }
   n38 = "Transform (Modify) Bone - Bone: dyn_hair_e_02_r"!(ComponentPose=@n37, bAlphaBoolEnabled:bool=True, Translation="0,0,0", Rotation="0,0,0", Scale="1,1,1") -> (Pose) {
    alphaInputType=Bool
    rotationMode=BMM_Additive
    translationMode=BMM_Additive
    alphaBoolBlend {blendOption=Linear}
    boneToModify {boneName=dyn_hair_e_02_r}
   }
   n39 = "Transform (Modify) Bone - Bone: dyn_hair_b_01_r"!(ComponentPose=@n38, bAlphaBoolEnabled:bool=True, Translation="0,0,0", Rotation="0,0,2", Scale="1,1,1") -> (Pose) {
    alphaInputType=Bool
    rotationMode=BMM_Additive
    translationMode=BMM_Additive
    alphaBoolBlend {blendOption=Linear}
    boneToModify {boneName=dyn_hair_b_01_r}
    rotation {roll=2}
   }
   Output Pose(Result=Component To Local(ComponentPose=@n39)) {layerGroup=None}
   Save cached pose 'InputPose'(Pose=Copy Pose From Mesh() {bUseAttachedParent=true}) {cacheName=InputPose}
  ]}
 EventGraph {nodes [
   Event BlueprintUpdateAnimation()
   > Sequence!(execute) -> (then_0, then_1, then_2:Exec, then_3:Exec, then_4:Exec, then_5:Exec, then_6:Exec, then_7:Exec, then_8:Exec)
   > n1 = Gravity Override State Function DEPRECATED(CharacterPartAnimInstance=Self-Reference(), GameTimeInSeconds=GetGameTimeInSeconds(), GravityOverrideParams=Get Gravity Override Params())
   > Set GravityOverride(GravityOverride=@n1)
   > Set Skydive(Skydive:double="To Float (Boolean)"(InBool=Get bIsSkyDiving())) -> (Output_Get:double)
  ]}
]
variables [
 Gravity Override Params {type=GravityOverrideParamsStruct}
 GravityOverride {type=Vector}
 Skydive {type=number}
]
